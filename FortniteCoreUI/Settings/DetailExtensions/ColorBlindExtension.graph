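# Settings detail panel extension that previews color-blind settings with Ishihara test plates and item rarity colors.
# A scaled-down grid shows six static plate images above five small weapon cards labeled from Low Rarity to High Rarity.
# On Construct, temporary item instances from CardItems fill the cards, with the first two indices swapped so Common precedes Uncommon.
# OnSettingAssigned has no logic.
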WidgetBlueprint /FortniteCoreUI/Settings/DetailExtensions/ColorBlindExtension
graphs [EventGraph {nodes [
   Event Construct()
   n1 = Get CardItems() -> (CardItems:Array of Fort Item Definition Object References)
   > SetItemToRepresent!(
    execute
    self:Fort Item Widget NUI Object Reference=Get icon2() -> (icon2:Fort Multi Size Item Card Object Reference)
    ItemToRepresent:Fort Item Object Reference=CreateTemporaryItemInstance!(
     self:Fort Item Definition Object Reference="Get (a copy)"(Array:Array of Fort Item Definition Object References=@n1) -> (Output:Fort Item Definition Object Reference)
     Count:int=1
     Level:int=-1
    ) -> (ReturnValue:Fort Item Object Reference)
   ) -> (then)
   > SetItemToRepresent(
    self=Get icon1() -> (icon1:Fort Multi Size Item Card Object Reference)
    ItemToRepresent=CreateTemporaryItemInstance(self="Get (a copy)"(Array=@n1, Dimension 1=1), Count=1, Level=-1)
   )
   > SetItemToRepresent(
    self=Get icon3() -> (icon3:Fort Multi Size Item Card Object Reference)
    ItemToRepresent=CreateTemporaryItemInstance(self="Get (a copy)"(Array=@n1, Dimension 1=2), Count=1, Level=-1)
   )
   > SetItemToRepresent(
    self=Get icon4() -> (icon4:Fort Multi Size Item Card Object Reference)
    ItemToRepresent=CreateTemporaryItemInstance(self="Get (a copy)"(Array=@n1, Dimension 1=3), Count=1, Level=-1)
   )
   > SetItemToRepresent(
    self=Get icon5() -> (icon5:Fort Multi Size Item Card Object Reference)
    ItemToRepresent=CreateTemporaryItemInstance(self="Get (a copy)"(Array=@n1, Dimension 1=4), Count=1, Level=-1)
   )
   Event OnSettingAssigned!() -> (OutputDelegate, then, InSetting:Fort Setting Object Reference) {is=title K2Node_Event}
  ]}]
variables [CardItems {
  type="array<FortItemDefinition>"
  default [
   /Game/Athena/Items/Weapons/WID_Assault_Auto_Athena_UC_Ore_T03
   /Game/Athena/Items/Weapons/WID_Assault_Auto_Athena_C_Ore_T02
   /Game/Athena/Items/Weapons/WID_Assault_Auto_Athena_R_Ore_T03
   /Game/Athena/Items/Weapons/WID_Assault_AutoHigh_Athena_VR_Ore_T03
   /Game/Athena/Items/Weapons/WID_Assault_AutoHigh_Athena_SR_Ore_T03
  ]
 }]
widgets {
 info {inheritedWidgetCount=0, parentClass=/Script/FortniteUI.FortSettingDetailExtension, rootWidgetClass=&VerticalBox, widgetCount=23}
 widgets [
  VerticalBox_2 {widgetClassPath=&VerticalBox, widget=VerticalBox_2}
  VerticalBox_0 {
   widgetClassPath=&VerticalBox
   parent=VerticalBox_2
   widget=VerticalBox_0
   slot {class=VerticalBoxSlot, properties {horizontalAlignment=HAlign_Left, padding=<ltrb 10 10 10 10>}}
  }
  ScaleBox_0 {
   widgetClassPath=&ScaleBox
   parent=VerticalBox_0
   widget=ScaleBox_0
   properties {stretch=UserSpecified, userSpecifiedScale=0.45}
   slot {class=VerticalBoxSlot}
  }
  Border_0 {widgetClassPath=&Border, parent=ScaleBox_0, widget=Border_0, properties {padding=<ltrb 0 0 0 0>}, slot {class=ScaleBoxSlot}}
  VerticalBox_3 {widgetClassPath=&VerticalBox, parent=Border_0, widget=VerticalBox_3, slot {class=BorderSlot, properties {padding=<ltrb 0 0 0 0>}}}
  UniformGridPanel_0 {widgetClassPath=/Script/UMG.UniformGridPanel, parent=VerticalBox_3, widget=UniformGridPanel_0, slot {class=VerticalBoxSlot}}
  VisibleToAll {
   widgetClassPath=&Image
   bIsVariable=true
   parent=UniformGridPanel_0
   widget=VisibleToAll
   properties {brush {
     bIsDynamicallyLoaded=false
     drawAs=Image
     imageType=NoImage
     mirroring=NoMirror
     resourceName=None
     resourceObject=/FortniteCoreUI/Settings/DetailExtensions/Accessibility/ishihara_all
     tiling=NoTile
     imageSize=<512 512>
     margin=<ltrb 0 0 0 0>
     outlineSettings {
      bUseBrushTransparency=false
      roundingType=HalfHeightRadius
      width=0
      color {colorUseRule=UseColor_Specified, specifiedColor=<rgba 0 0 0 0>}
      cornerRadii=<0 0 0 0>
     }
     tintColor {colorUseRule=UseColor_Specified, specifiedColor=<rgba 1 1 1 1>}
     uVRegion {bIsValid=false, max=<0 0>, min=<0 0>}
    }}
   slot {class=UniformGridSlot}
  }
  GreenRed1 {
   widgetClassPath=&Image
   bIsVariable=true
   parent=UniformGridPanel_0
   widget=GreenRed1
   properties {brush {
     bIsDynamicallyLoaded=false
     drawAs=Image
     imageType=NoImage
     mirroring=NoMirror
     resourceName=None
     resourceObject=/FortniteCoreUI/Settings/DetailExtensions/Accessibility/ishihara_g1
     tiling=NoTile
     imageSize=<512 512>
     margin=<ltrb 0 0 0 0>
     outlineSettings {
      bUseBrushTransparency=false
      roundingType=HalfHeightRadius
      width=0
      color {colorUseRule=UseColor_Specified, specifiedColor=<rgba 0 0 0 0>}
      cornerRadii=<0 0 0 0>
     }
     tintColor {colorUseRule=UseColor_Specified, specifiedColor=<rgba 1 1 1 1>}
     uVRegion {bIsValid=false, max=<0 0>, min=<0 0>}
    }}
   slot {class=UniformGridSlot, properties {column=1}}
  }
  GreenRed2 {
   widgetClassPath=&Image
   bIsVariable=true
   parent=UniformGridPanel_0
   widget=GreenRed2
   properties {brush {
     bIsDynamicallyLoaded=false
     drawAs=Image
     imageType=NoImage
     mirroring=NoMirror
     resourceName=None
     resourceObject=/FortniteCoreUI/Settings/DetailExtensions/Accessibility/ishihara_g2
     tiling=NoTile
     imageSize=<512 512>
     margin=<ltrb 0 0 0 0>
     outlineSettings {
      bUseBrushTransparency=false
      roundingType=HalfHeightRadius
      width=0
      color {colorUseRule=UseColor_Specified, specifiedColor=<rgba 0 0 0 0>}
      cornerRadii=<0 0 0 0>
     }
     tintColor {colorUseRule=UseColor_Specified, specifiedColor=<rgba 1 1 1 1>}
     uVRegion {bIsValid=false, max=<0 0>, min=<0 0>}
    }}
   slot {class=UniformGridSlot, properties {column=2}}
  }
  GreenRed3 {
   widgetClassPath=&Image
   bIsVariable=true
   parent=UniformGridPanel_0
   widget=GreenRed3
   properties {brush {
     bIsDynamicallyLoaded=false
     drawAs=Image
     imageType=NoImage
     mirroring=NoMirror
     resourceName=None
     resourceObject=/FortniteCoreUI/Settings/DetailExtensions/Accessibility/ishihara_g3
     tiling=NoTile
     imageSize=<512 512>
     margin=<ltrb 0 0 0 0>
     outlineSettings {
      bUseBrushTransparency=false
      roundingType=HalfHeightRadius
      width=0
      color {colorUseRule=UseColor_Specified, specifiedColor=<rgba 0 0 0 0>}
      cornerRadii=<0 0 0 0>
     }
     tintColor {colorUseRule=UseColor_Specified, specifiedColor=<rgba 1 1 1 1>}
     uVRegion {bIsValid=false, max=<0 0>, min=<0 0>}
    }}
   slot {class=UniformGridSlot, properties {row=1}}
  }
  Blue1 {
   widgetClassPath=&Image
   bIsVariable=true
   parent=UniformGridPanel_0
   widget=Blue1
   properties {brush {
     bIsDynamicallyLoaded=false
     drawAs=Image
     imageType=NoImage
     mirroring=NoMirror
     resourceName=None
     resourceObject=/FortniteCoreUI/Settings/DetailExtensions/Accessibility/ishihara_protanoomaly
     tiling=NoTile
     imageSize=<512 512>
     margin=<ltrb 0 0 0 0>
     outlineSettings {
      bUseBrushTransparency=false
      roundingType=HalfHeightRadius
      width=0
      color {colorUseRule=UseColor_Specified, specifiedColor=<rgba 0 0 0 0>}
      cornerRadii=<0 0 0 0>
     }
     tintColor {colorUseRule=UseColor_Specified, specifiedColor=<rgba 1 1 1 1>}
     uVRegion {bIsValid=false, max=<0 0>, min=<0 0>}
    }}
   slot {class=UniformGridSlot, properties {column=1, row=1}}
  }
  Blue2 {
   widgetClassPath=&Image
   bIsVariable=true
   parent=UniformGridPanel_0
   widget=Blue2
   properties {brush {
     bIsDynamicallyLoaded=false
     drawAs=Image
     imageType=NoImage
     mirroring=NoMirror
     resourceName=None
     resourceObject=/FortniteCoreUI/Settings/DetailExtensions/Accessibility/ishihara_protanopia
     tiling=NoTile
     imageSize=<512 512>
     margin=<ltrb 0 0 0 0>
     outlineSettings {
      bUseBrushTransparency=false
      roundingType=HalfHeightRadius
      width=0
      color {colorUseRule=UseColor_Specified, specifiedColor=<rgba 0 0 0 0>}
      cornerRadii=<0 0 0 0>
     }
     tintColor {colorUseRule=UseColor_Specified, specifiedColor=<rgba 1 1 1 1>}
     uVRegion {bIsValid=false, max=<0 0>, min=<0 0>}
    }}
   slot {class=UniformGridSlot, properties {column=2, row=1}}
  }
  SizeBox_1 {
   widgetClassPath=&SizeBox
   parent=VerticalBox_0
   widget=SizeBox_1
   properties {bOverride_WidthOverride=true, widthOverride=668}
   slot {class=VerticalBoxSlot, properties {verticalAlignment=VAlign_Top, padding=<ltrb 0 10 0 0>}}
  }
  VerticalBox_4 {widgetClassPath=&VerticalBox, parent=SizeBox_1, widget=VerticalBox_4, slot {class=SizeBoxSlot}}
  HorizontalBox_1 {widgetClassPath=&HorizontalBox, parent=VerticalBox_4, widget=HorizontalBox_1, slot {class=VerticalBoxSlot}}
  icon1 {
   widgetClassPath=/Script/FortniteUI.FortMultiSizeItemCard
   bIsVariable=true
   parent=HorizontalBox_1
   widget=icon1
   properties {bPItemCardSize=S, shouldCollapseBorderPadding=true, shouldSuppressItemTypeIcon=true}
   slot {class=HorizontalBoxSlot, properties {horizontalAlignment=HAlign_Center, padding=<ltrb 0 0 8 0>, size {sizeRule=Fill, value=1}}}
  }
  icon2 {
   widgetClassPath=&FortMultiSizeItemCard
   bIsVariable=true
   parent=HorizontalBox_1
   widget=icon2
   properties {bPItemCardSize=S, shouldCollapseBorderPadding=true, shouldDisplayItemAsReward=true, shouldSuppressItemTypeIcon=true}
   slot {class=HorizontalBoxSlot, properties {horizontalAlignment=HAlign_Center, padding=<ltrb 0 0 8 0>, size {sizeRule=Fill, value=1}}}
  }
  icon3 {
   widgetClassPath=&FortMultiSizeItemCard
   bIsVariable=true
   parent=HorizontalBox_1
   widget=icon3
   properties {bPItemCardSize=S, shouldCollapseBorderPadding=true, shouldDisplayItemAsReward=true, shouldSuppressItemTypeIcon=true}
   slot {class=HorizontalBoxSlot, properties {horizontalAlignment=HAlign_Center, padding=<ltrb 0 0 8 0>, size {sizeRule=Fill, value=1}}}
  }
  icon4 {
   widgetClassPath=&FortMultiSizeItemCard
   bIsVariable=true
   parent=HorizontalBox_1
   widget=icon4
   properties {bPItemCardSize=S, shouldCollapseBorderPadding=true, shouldDisplayItemAsReward=true, shouldSuppressItemTypeIcon=true}
   slot {class=HorizontalBoxSlot, properties {horizontalAlignment=HAlign_Center, padding=<ltrb 0 0 8 0>, size {sizeRule=Fill, value=1}}}
  }
  icon5 {
   widgetClassPath=&FortMultiSizeItemCard
   bIsVariable=true
   parent=HorizontalBox_1
   widget=icon5
   properties {bPItemCardSize=S, shouldCollapseBorderPadding=true, shouldDisplayItemAsReward=true, shouldSuppressItemTypeIcon=true}
   slot {class=HorizontalBoxSlot, properties {horizontalAlignment=HAlign_Center, size {sizeRule=Fill, value=1}}}
  }
  Overlay_0 {
   widgetClassPath=&Overlay
   parent=VerticalBox_4
   widget=Overlay_0
   slot {class=VerticalBoxSlot, properties {padding=<ltrb 0 6 0 0>}}
  }
  CommonTextBlock_3 {
   widgetClassPath=&CommonTextBlock
   parent=Overlay_0
   widget=CommonTextBlock_3
   properties {
    style=/FortUILibrary/Styles/Typography/Foundation/TextStyle-Base-M-B-S.TextStyle-Base-M-B-S_C
    text=Low Rarity
    font {
     bForceMonospaced=false
     bMaterialIsStencil=false
     fontMaterial=null
     fontObject=&NotoSans
     letterSpacing=0
     monospacedWidth=1
     size=24
     skewAmount=0
     typefaceFontName=NotoSans-Bold
     outlineSettings {
      bApplyOutlineToDropShadows=true
      bMiteredCorners=false
      bSeparateFillAlpha=true
      outlineMaterial=&M-FontShadow
      outlineSize=3
      outlineColor=<rgba 0 0 0 0.5>
     }
    }
    shadowOffset=<0 0>
   }
   slot {class=OverlaySlot}
  }
  CommonTextBlock_4 {
   widgetClassPath=&CommonTextBlock
   parent=Overlay_0
   widget=CommonTextBlock_4
   properties {
    style=&TextStyle-Base-M-B-S_C
    text=High Rarity
    font {
     bForceMonospaced=false
     bMaterialIsStencil=false
     fontMaterial=null
     fontObject=&NotoSans
     letterSpacing=0
     monospacedWidth=1
     size=24
     skewAmount=0
     typefaceFontName=NotoSans-Bold
     outlineSettings {
      bApplyOutlineToDropShadows=true
      bMiteredCorners=false
      bSeparateFillAlpha=true
      outlineMaterial=&M-FontShadow
      outlineSize=3
      outlineColor=<rgba 0 0 0 0.5>
     }
    }
    shadowOffset=<0 0>
   }
   slot {class=OverlaySlot, properties {horizontalAlignment=HAlign_Right}}
  }
 ]
}
