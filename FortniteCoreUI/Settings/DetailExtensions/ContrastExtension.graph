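# ContrastExtension is a settings detail extension widget (child of FortSettingDetailExtension) that shows a contrast reference made of colored bands.
# A vertical box stacks seven images that all draw the same ContrastBands resource as horizontal strips.
# Color tints give blue, red, green, yellow, cyan and magenta strips, plus one left untinted.
# The layout is purely static, with an empty event graph and no variables of its own.
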WidgetBlueprint /FortniteCoreUI/Settings/DetailExtensions/ContrastExtension
graphs [EventGraph {links [], nodes []}]
variables []
widgets {
 info {inheritedWidgetCount=0, parentClass=/Script/FortniteUI.FortSettingDetailExtension, rootWidgetClass=&VerticalBox, widgetCount=8}
 widgets [
  VerticalBox_19 {widgetClassPath=&VerticalBox, widget=VerticalBox_19}
  Image_47 {
   widgetClassPath=&Image
   bIsVariable=true
   parent=VerticalBox_19
   widget=Image_47
   properties {
    brush {
     bIsDynamicallyLoaded=false
     drawAs=Image
     imageType=NoImage
     mirroring=NoMirror
     resourceName=None
     resourceObject=&/ContrastBands
     tiling=NoTile
     imageSize=<512 32>
     margin=<ltrb 0 0 0 0>
     outlineSettings {
      bUseBrushTransparency=false
      roundingType=HalfHeightRadius
      width=0
      color {colorUseRule=UseColor_Specified, specifiedColor=<rgba 0 0 0 0>}
      cornerRadii=<0 0 0 0>
     }
     tintColor {colorUseRule=UseColor_Specified, specifiedColor=<rgba 1 1 1 1>}
     uVRegion {bIsValid=false, max=<0 0>, min=<0 0>}
    }
    colorAndOpacity=<rgba 0 0 1 1>
   }
   slot {class=VerticalBoxSlot}
  }
  Image {
   widgetClassPath=&Image
   bIsVariable=true
   parent=VerticalBox_19
   widget=Image
   properties {
    brush {
     bIsDynamicallyLoaded=false
     drawAs=Image
     imageType=NoImage
     mirroring=NoMirror
     resourceName=None
     resourceObject=&ContrastBands
     tiling=NoTile
     imageSize=<512 32>
     margin=<ltrb 0 0 0 0>
     outlineSettings {
      bUseBrushTransparency=false
      roundingType=HalfHeightRadius
      width=0
      color {colorUseRule=UseColor_Specified, specifiedColor=<rgba 0 0 0 0>}
      cornerRadii=<0 0 0 0>
     }
     tintColor {colorUseRule=UseColor_Specified, specifiedColor=<rgba 1 1 1 1>}
     uVRegion {bIsValid=false, max=<0 0>, min=<0 0>}
    }
    colorAndOpacity=<rgba 1 0 0 1>
   }
   slot {class=VerticalBoxSlot}
  }
  Image_1 {
   widgetClassPath=&Image
   bIsVariable=true
   parent=VerticalBox_19
   widget=Image_1
   properties {
    brush {
     bIsDynamicallyLoaded=false
     drawAs=Image
     imageType=NoImage
     mirroring=NoMirror
     resourceName=None
     resourceObject=&ContrastBands
     tiling=NoTile
     imageSize=<512 32>
     margin=<ltrb 0 0 0 0>
     outlineSettings {
      bUseBrushTransparency=false
      roundingType=HalfHeightRadius
      width=0
      color {colorUseRule=UseColor_Specified, specifiedColor=<rgba 0 0 0 0>}
      cornerRadii=<0 0 0 0>
     }
     tintColor {colorUseRule=UseColor_Specified, specifiedColor=<rgba 1 1 1 1>}
     uVRegion {bIsValid=false, max=<0 0>, min=<0 0>}
    }
    colorAndOpacity=<rgba 0 1 0 1>
   }
   slot {class=VerticalBoxSlot}
  }
  Image_2 {
   widgetClassPath=&Image
   bIsVariable=true
   parent=VerticalBox_19
   widget=Image_2
   properties {
    brush {
     bIsDynamicallyLoaded=false
     drawAs=Image
     imageType=NoImage
     mirroring=NoMirror
     resourceName=None
     resourceObject=&ContrastBands
     tiling=NoTile
     imageSize=<512 32>
     margin=<ltrb 0 0 0 0>
     outlineSettings {
      bUseBrushTransparency=false
      roundingType=HalfHeightRadius
      width=0
      color {colorUseRule=UseColor_Specified, specifiedColor=<rgba 0 0 0 0>}
      cornerRadii=<0 0 0 0>
     }
     tintColor {colorUseRule=UseColor_Specified, specifiedColor=<rgba 1 1 1 1>}
     uVRegion {bIsValid=false, max=<0 0>, min=<0 0>}
    }
    colorAndOpacity=<rgba 1 1 0 1>
   }
   slot {class=VerticalBoxSlot}
  }
  Image_3 {
   widgetClassPath=&Image
   bIsVariable=true
   parent=VerticalBox_19
   widget=Image_3
   properties {
    brush {
     bIsDynamicallyLoaded=false
     drawAs=Image
     imageType=NoImage
     mirroring=NoMirror
     resourceName=None
     resourceObject=&ContrastBands
     tiling=NoTile
     imageSize=<512 32>
     margin=<ltrb 0 0 0 0>
     outlineSettings {
      bUseBrushTransparency=false
      roundingType=HalfHeightRadius
      width=0
      color {colorUseRule=UseColor_Specified, specifiedColor=<rgba 0 0 0 0>}
      cornerRadii=<0 0 0 0>
     }
     tintColor {colorUseRule=UseColor_Specified, specifiedColor=<rgba 1 1 1 1>}
     uVRegion {bIsValid=false, max=<0 0>, min=<0 0>}
    }
    colorAndOpacity=<rgba 0 1 1 1>
   }
   slot {class=VerticalBoxSlot}
  }
  Image_4 {
   widgetClassPath=&Image
   bIsVariable=true
   parent=VerticalBox_19
   widget=Image_4
   properties {
    brush {
     bIsDynamicallyLoaded=false
     drawAs=Image
     imageType=NoImage
     mirroring=NoMirror
     resourceName=None
     resourceObject=&ContrastBands
     tiling=NoTile
     imageSize=<512 32>
     margin=<ltrb 0 0 0 0>
     outlineSettings {
      bUseBrushTransparency=false
      roundingType=HalfHeightRadius
      width=0
      color {colorUseRule=UseColor_Specified, specifiedColor=<rgba 0 0 0 0>}
      cornerRadii=<0 0 0 0>
     }
     tintColor {colorUseRule=UseColor_Specified, specifiedColor=<rgba 1 1 1 1>}
     uVRegion {bIsValid=false, max=<0 0>, min=<0 0>}
    }
    colorAndOpacity=<rgba 1 0 1 1>
   }
   slot {class=VerticalBoxSlot}
  }
  Image_5 {
   widgetClassPath=&Image
   bIsVariable=true
   parent=VerticalBox_19
   widget=Image_5
   properties {brush {
     bIsDynamicallyLoaded=false
     drawAs=Image
     imageType=NoImage
     mirroring=NoMirror
     resourceName=None
     resourceObject=&ContrastBands
     tiling=NoTile
     imageSize=<512 32>
     margin=<ltrb 0 0 0 0>
     outlineSettings {
      bUseBrushTransparency=false
      roundingType=HalfHeightRadius
      width=0
      color {colorUseRule=UseColor_Specified, specifiedColor=<rgba 0 0 0 0>}
      cornerRadii=<0 0 0 0>
     }
     tintColor {colorUseRule=UseColor_Specified, specifiedColor=<rgba 1 1 1 1>}
     uVRegion {bIsValid=false, max=<0 0>, min=<0 0>}
    }}
   slot {class=VerticalBoxSlot}
  }
 ]
}
