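# Controller diagram widget for the Erebus (Switch) gamepad in the settings detail extensions, built on FortControllerHighlightDiagram.
# A base controller image sits under a visibility switcher holding the inherited CanvasPanel_Highlights.
# That canvas point-anchors one collapsed HUD thumbstick-material image per input, named by gamepad control, over its spot on the art.
# The event graph has only empty stubs, so any highlight logic comes from the parent class.
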WidgetBlueprint /FortniteCoreUI/Settings/DetailExtensions/ControllerDiagrams/WBP_ControllerDiagram_Erebus
graphs [EventGraph {nodes [
   Event PreConstruct()
   Event Construct()
   Event Tick()
  ]}]
variables []
widgets {
 info {inheritedWidgetCount=1, parentClass=/Script/FortniteUI.FortControllerHighlightDiagram, rootWidgetClass=&SizeBox, widgetCount=21}
 widgets [
  SizeBox_0 {widgetClassPath=&SizeBox, widget=SizeBox_0}
  Overlay_30 {widgetClassPath=&Overlay, parent=SizeBox_0, widget=Overlay_30, slot {class=SizeBoxSlot}}
  Image_ControllerInput {
   widgetClassPath=&Image
   bIsVariable=true
   parent=Overlay_30
   widget=Image_ControllerInput
   properties {
    bIsEnabled=false
    visibility=SelfHitTestInvisible
    brush {
     bIsDynamicallyLoaded=false
     drawAs=Image
     imageType=NoImage
     mirroring=NoMirror
     resourceName=None
     resourceObject=/FortUILibrary/Input/GamepadSwitch/T_UI_Erebus_ControllerConfig
     tiling=NoTile
     imageSize=<1024 672>
     margin=<ltrb 0 0 0 0>
     outlineSettings {
      bUseBrushTransparency=false
      roundingType=HalfHeightRadius
      width=0
      color {colorUseRule=UseColor_Specified, specifiedColor=<rgba 0 0 0 0>}
      cornerRadii=<0 0 0 0>
     }
     tintColor {colorUseRule=UseColor_Specified, specifiedColor=<rgba 1 1 1 1>}
     uVRegion {bIsValid=false, max=<0 0>, min=<0 0>}
    }
   }
   slot {class=OverlaySlot, properties {horizontalAlignment=HAlign_Fill, verticalAlignment=VAlign_Fill}}
  }
  Switcher_HighlightedElement {
   widgetClassPath=&CommonVisibilitySwitcher
   bIsVariable=true
   parent=Overlay_30
   widget=Switcher_HighlightedElement
   slot {class=OverlaySlot, properties {horizontalAlignment=HAlign_Fill, verticalAlignment=VAlign_Fill}}
  }
  CanvasPanel_Highlights {
   widgetClassPath=&CanvasPanel
   bInherited=true
   parent=Switcher_HighlightedElement
   widget=CanvasPanel_Highlights
   slot {class=CommonVisibilitySwitcherSlot}
  }
  Gamepad_FaceButton_Right {
   widgetClassPath=&Image
   bIsVariable=true
   parent=CanvasPanel_Highlights
   widget=Gamepad_FaceButton_Right
   properties {
    bIsEnabled=false
    visibility=Collapsed
    brush {
     bIsDynamicallyLoaded=false
     drawAs=Image
     imageType=NoImage
     mirroring=NoMirror
     resourceName=None
     resourceObject=/Game/Athena/HUD/Art/Materials/M_HUD_Thumbstick_Wide
     tiling=NoTile
     imageSize=<32 32>
     margin=<ltrb 0 0 0 0>
     outlineSettings {
      bUseBrushTransparency=false
      roundingType=HalfHeightRadius
      width=0
      color {colorUseRule=UseColor_Specified, specifiedColor=<rgba 0 0 0 0>}
      cornerRadii=<0 0 0 0>
     }
     tintColor {colorUseRule=UseColor_Specified, specifiedColor=<rgba 1 1 1 1>}
     uVRegion {bIsValid=false, max=<0 0>, min=<0 0>}
    }
   }
   slot {
    class=CanvasPanelSlot
    properties {layoutData {alignment=<0 0>, anchors {maximum=<0.668 0.41>, minimum=<0.668 0.41>}, offsets=<ltrb -26.6521 -26.8 52 52>}}
   }
  }
  Gamepad_FaceButton_Left {
   widgetClassPath=&Image
   bIsVariable=true
   parent=CanvasPanel_Highlights
   widget=Gamepad_FaceButton_Left
   properties {
    bIsEnabled=false
    visibility=Collapsed
    brush {
     bIsDynamicallyLoaded=false
     drawAs=Image
     imageType=NoImage
     mirroring=NoMirror
     resourceName=None
     resourceObject=&M_HUD_Thumbstick_Wide
     tiling=NoTile
     imageSize=<64 64>
     margin=<ltrb 0 0 0 0>
     outlineSettings {
      bUseBrushTransparency=false
      roundingType=HalfHeightRadius
      width=0
      color {colorUseRule=UseColor_Specified, specifiedColor=<rgba 0 0 0 0>}
      cornerRadii=<0 0 0 0>
     }
     tintColor {colorUseRule=UseColor_Specified, specifiedColor=<rgba 1 1 1 1>}
     uVRegion {bIsValid=false, max=<0 0>, min=<0 0>}
    }
   }
   slot {
    class=CanvasPanelSlot
    properties {layoutData {alignment=<0 0>, anchors {maximum=<0.608 0.4121>, minimum=<0.608 0.4121>}, offsets=<ltrb -27.36 -29.0788 52 52>}}
   }
  }
  Gamepad_FaceButton_Top {
   widgetClassPath=&Image
   bIsVariable=true
   parent=CanvasPanel_Highlights
   widget=Gamepad_FaceButton_Top
   properties {
    bIsEnabled=false
    visibility=Collapsed
    brush {
     bIsDynamicallyLoaded=false
     drawAs=Image
     imageType=NoImage
     mirroring=NoMirror
     resourceName=None
     resourceObject=&M_HUD_Thumbstick_Wide
     tiling=NoTile
     imageSize=<64 64>
     margin=<ltrb 0 0 0 0>
     outlineSettings {
      bUseBrushTransparency=false
      roundingType=HalfHeightRadius
      width=0
      color {colorUseRule=UseColor_Specified, specifiedColor=<rgba 0 0 0 0>}
      cornerRadii=<0 0 0 0>
     }
     tintColor {colorUseRule=UseColor_Specified, specifiedColor=<rgba 1 1 1 1>}
     uVRegion {bIsValid=false, max=<0 0>, min=<0 0>}
    }
   }
   slot {
    class=CanvasPanelSlot
    properties {layoutData {alignment=<0 0>, anchors {maximum=<0.6385 0.3686>, minimum=<0.6385 0.3686>}, offsets=<ltrb -25.9988 -26.0675 52 52>}}
   }
  }
  Gamepad_FaceButton_Bottom {
   widgetClassPath=&Image
   bIsVariable=true
   parent=CanvasPanel_Highlights
   widget=Gamepad_FaceButton_Bottom
   properties {
    bIsEnabled=false
    visibility=Collapsed
    brush {
     bIsDynamicallyLoaded=false
     drawAs=Image
     imageType=NoImage
     mirroring=NoMirror
     resourceName=None
     resourceObject=&M_HUD_Thumbstick_Wide
     tiling=NoTile
     imageSize=<64 64>
     margin=<ltrb 0 0 0 0>
     outlineSettings {
      bUseBrushTransparency=false
      roundingType=HalfHeightRadius
      width=0
      color {colorUseRule=UseColor_Specified, specifiedColor=<rgba 0 0 0 0>}
      cornerRadii=<0 0 0 0>
     }
     tintColor {colorUseRule=UseColor_Specified, specifiedColor=<rgba 1 1 1 1>}
     uVRegion {bIsValid=false, max=<0 0>, min=<0 0>}
    }
   }
   slot {
    class=CanvasPanelSlot
    properties {layoutData {alignment=<0 0>, anchors {maximum=<0.6384 0.454>, minimum=<0.6384 0.454>}, offsets=<ltrb -25.6704 -26.3362 52 52>}}
   }
  }
  Gamepad_Special_Right {
   widgetClassPath=&Image
   bIsVariable=true
   parent=CanvasPanel_Highlights
   widget=Gamepad_Special_Right
   properties {
    bIsEnabled=false
    visibility=Collapsed
    brush {
     bIsDynamicallyLoaded=false
     drawAs=Image
     imageType=NoImage
     mirroring=NoMirror
     resourceName=None
     resourceObject=/Game/Athena/HUD/Art/Materials/M_HUD_Thumbstick
     tiling=NoTile
     imageSize=<32 32>
     margin=<ltrb 0 0 0 0>
     outlineSettings {
      bUseBrushTransparency=false
      roundingType=HalfHeightRadius
      width=0
      color {colorUseRule=UseColor_Specified, specifiedColor=<rgba 0 0 0 0>}
      cornerRadii=<0 0 0 0>
     }
     tintColor {colorUseRule=UseColor_Specified, specifiedColor=<rgba 1 1 1 1>}
     uVRegion {bIsValid=false, max=<0 0>, min=<0 0>}
    }
   }
   slot {
    class=CanvasPanelSlot
    properties {layoutData {alignment=<0 0>, anchors {maximum=<0.6 0.3205>, minimum=<0.6 0.3205>}, offsets=<ltrb -16 -14.1 32 32>}}
   }
  }
  Gamepad_Special_Left {
   widgetClassPath=&Image
   bIsVariable=true
   parent=CanvasPanel_Highlights
   widget=Gamepad_Special_Left
   properties {
    bIsEnabled=false
    visibility=Collapsed
    brush {
     bIsDynamicallyLoaded=false
     drawAs=Image
     imageType=NoImage
     mirroring=NoMirror
     resourceName=None
     resourceObject=&M_HUD_Thumbstick
     tiling=NoTile
     imageSize=<32 32>
     margin=<ltrb 0 0 0 0>
     outlineSettings {
      bUseBrushTransparency=false
      roundingType=HalfHeightRadius
      width=0
      color {colorUseRule=UseColor_Specified, specifiedColor=<rgba 0 0 0 0>}
      cornerRadii=<0 0 0 0>
     }
     tintColor {colorUseRule=UseColor_Specified, specifiedColor=<rgba 1 1 1 1>}
     uVRegion {bIsValid=false, max=<0 0>, min=<0 0>}
    }
   }
   slot {
    class=CanvasPanelSlot
    properties {layoutData {alignment=<0 0>, anchors {maximum=<0.4 0.3206>, minimum=<0.4 0.3206>}, offsets=<ltrb -16 -18.2588 32 32>}}
   }
  }
  Gamepad_LeftThumbstick {
   widgetClassPath=&Image
   bIsVariable=true
   parent=CanvasPanel_Highlights
   widget=Gamepad_LeftThumbstick
   properties {
    bIsEnabled=false
    visibility=Collapsed
    brush {
     bIsDynamicallyLoaded=false
     drawAs=Image
     imageType=NoImage
     mirroring=NoMirror
     resourceName=None
     resourceObject=&M_HUD_Thumbstick
     tiling=NoTile
     imageSize=<32 32>
     margin=<ltrb 0 0 0 0>
     outlineSettings {
      bUseBrushTransparency=false
      roundingType=HalfHeightRadius
      width=0
      color {colorUseRule=UseColor_Specified, specifiedColor=<rgba 0 0 0 0>}
      cornerRadii=<0 0 0 0>
     }
     tintColor {colorUseRule=UseColor_Specified, specifiedColor=<rgba 1 1 1 1>}
     uVRegion {bIsValid=false, max=<0 0>, min=<0 0>}
    }
   }
   slot {
    class=CanvasPanelSlot
    properties {layoutData {alignment=<0 0>, anchors {maximum=<0.3629 0.411>, minimum=<0.3629 0.411>}, offsets=<ltrb -28.8467 -31.8886 60 60>}}
   }
  }
  Gamepad_RightThumbstick {
   widgetClassPath=&Image
   bIsVariable=true
   parent=CanvasPanel_Highlights
   widget=Gamepad_RightThumbstick
   properties {
    bIsEnabled=false
    visibility=Collapsed
    brush {
     bIsDynamicallyLoaded=false
     drawAs=Image
     imageType=NoImage
     mirroring=NoMirror
     resourceName=None
     resourceObject=&M_HUD_Thumbstick
     tiling=NoTile
     imageSize=<32 32>
     margin=<ltrb 0 0 0 0>
     outlineSettings {
      bUseBrushTransparency=false
      roundingType=HalfHeightRadius
      width=0
      color {colorUseRule=UseColor_Specified, specifiedColor=<rgba 0 0 0 0>}
      cornerRadii=<0 0 0 0>
     }
     tintColor {colorUseRule=UseColor_Specified, specifiedColor=<rgba 1 1 1 1>}
     uVRegion {bIsValid=false, max=<0 0>, min=<0 0>}
    }
   }
   slot {
    class=CanvasPanelSlot
    properties {layoutData {alignment=<0 0>, anchors {maximum=<0.6389 0.5746>, minimum=<0.6389 0.5746>}, offsets=<ltrb -30.7245 -32.527 60 60>}}
   }
  }
  Gamepad_DPad_Left {
   widgetClassPath=&Image
   bIsVariable=true
   parent=CanvasPanel_Highlights
   widget=Gamepad_DPad_Left
   properties {
    bIsEnabled=false
    visibility=Collapsed
    brush {
     bIsDynamicallyLoaded=false
     drawAs=Image
     imageType=NoImage
     mirroring=NoMirror
     resourceName=None
     resourceObject=&M_HUD_Thumbstick_Wide
     tiling=NoTile
     imageSize=<32 32>
     margin=<ltrb 0 0 0 0>
     outlineSettings {
      bUseBrushTransparency=false
      roundingType=HalfHeightRadius
      width=0
      color {colorUseRule=UseColor_Specified, specifiedColor=<rgba 0 0 0 0>}
      cornerRadii=<0 0 0 0>
     }
     tintColor {colorUseRule=UseColor_Specified, specifiedColor=<rgba 1 1 1 1>}
     uVRegion {bIsValid=false, max=<0 0>, min=<0 0>}
    }
   }
   slot {
    class=CanvasPanelSlot
    properties {layoutData {alignment=<0 0>, anchors {maximum=<0.3326 0.5728>, minimum=<0.3326 0.5728>}, offsets=<ltrb -26.519 -26.6078 52 52>}}
   }
  }
  Gamepad_DPad_Right {
   widgetClassPath=&Image
   bIsVariable=true
   parent=CanvasPanel_Highlights
   widget=Gamepad_DPad_Right
   properties {
    bIsEnabled=false
    visibility=Collapsed
    brush {
     bIsDynamicallyLoaded=false
     drawAs=Image
     imageType=NoImage
     mirroring=NoMirror
     resourceName=None
     resourceObject=&M_HUD_Thumbstick_Wide
     tiling=NoTile
     imageSize=<32 32>
     margin=<ltrb 0 0 0 0>
     outlineSettings {
      bUseBrushTransparency=false
      roundingType=HalfHeightRadius
      width=0
      color {colorUseRule=UseColor_Specified, specifiedColor=<rgba 0 0 0 0>}
      cornerRadii=<0 0 0 0>
     }
     tintColor {colorUseRule=UseColor_Specified, specifiedColor=<rgba 1 1 1 1>}
     uVRegion {bIsValid=false, max=<0 0>, min=<0 0>}
    }
   }
   slot {
    class=CanvasPanelSlot
    properties {layoutData {alignment=<0 0>, anchors {maximum=<0.3925 0.57>, minimum=<0.3925 0.57>}, offsets=<ltrb -25.6 -23.6508 52 52>}}
   }
  }
  Gamepad_DPad_Up {
   widgetClassPath=&Image
   bIsVariable=true
   parent=CanvasPanel_Highlights
   widget=Gamepad_DPad_Up
   properties {
    bIsEnabled=false
    visibility=Collapsed
    brush {
     bIsDynamicallyLoaded=false
     drawAs=Image
     imageType=NoImage
     mirroring=NoMirror
     resourceName=None
     resourceObject=&M_HUD_Thumbstick_Wide
     tiling=NoTile
     imageSize=<32 32>
     margin=<ltrb 0 0 0 0>
     outlineSettings {
      bUseBrushTransparency=false
      roundingType=HalfHeightRadius
      width=0
      color {colorUseRule=UseColor_Specified, specifiedColor=<rgba 0 0 0 0>}
      cornerRadii=<0 0 0 0>
     }
     tintColor {colorUseRule=UseColor_Specified, specifiedColor=<rgba 1 1 1 1>}
     uVRegion {bIsValid=false, max=<0 0>, min=<0 0>}
    }
   }
   slot {
    class=CanvasPanelSlot
    properties {layoutData {alignment=<0 0>, anchors {maximum=<0.3629 0.5285>, minimum=<0.3629 0.5285>}, offsets=<ltrb -24.8295 -26.7552 52 52>}}
   }
  }
  Gamepad_DPad_Down {
   widgetClassPath=&Image
   bIsVariable=true
   parent=CanvasPanel_Highlights
   widget=Gamepad_DPad_Down
   properties {
    bIsEnabled=false
    visibility=Collapsed
    brush {
     bIsDynamicallyLoaded=false
     drawAs=Image
     imageType=NoImage
     mirroring=NoMirror
     resourceName=None
     resourceObject=&M_HUD_Thumbstick_Wide
     tiling=NoTile
     imageSize=<32 32>
     margin=<ltrb 0 0 0 0>
     outlineSettings {
      bUseBrushTransparency=false
      roundingType=HalfHeightRadius
      width=0
      color {colorUseRule=UseColor_Specified, specifiedColor=<rgba 0 0 0 0>}
      cornerRadii=<0 0 0 0>
     }
     tintColor {colorUseRule=UseColor_Specified, specifiedColor=<rgba 1 1 1 1>}
     uVRegion {bIsValid=false, max=<0 0>, min=<0 0>}
    }
   }
   slot {
    class=CanvasPanelSlot
    properties {layoutData {alignment=<0 0>, anchors {maximum=<0.3637 0.6175>, minimum=<0.3637 0.6175>}, offsets=<ltrb -26.2675 -26.8892 52 52>}}
   }
  }
  Gamepad_LeftShoulder {
   widgetClassPath=&Image
   bIsVariable=true
   parent=CanvasPanel_Highlights
   widget=Gamepad_LeftShoulder
   properties {
    bIsEnabled=false
    visibility=Collapsed
    brush {
     bIsDynamicallyLoaded=false
     drawAs=Image
     imageType=NoImage
     mirroring=NoMirror
     resourceName=None
     resourceObject=&M_HUD_Thumbstick
     tiling=NoTile
     imageSize=<32 32>
     margin=<ltrb 0 0 0 0>
     outlineSettings {
      bUseBrushTransparency=false
      roundingType=HalfHeightRadius
      width=0
      color {colorUseRule=UseColor_Specified, specifiedColor=<rgba 0 0 0 0>}
      cornerRadii=<0 0 0 0>
     }
     tintColor {colorUseRule=UseColor_Specified, specifiedColor=<rgba 1 1 1 1>}
     uVRegion {bIsValid=false, max=<0 0>, min=<0 0>}
    }
   }
   slot {
    class=CanvasPanelSlot
    properties {layoutData {alignment=<0 0>, anchors {maximum=<0.3486 0.2671>, minimum=<0.3486 0.2671>}, offsets=<ltrb -29.3734 -20.5004 64 32>}}
   }
  }
  Gamepad_LeftTrigger {
   widgetClassPath=&Image
   bIsVariable=true
   parent=CanvasPanel_Highlights
   widget=Gamepad_LeftTrigger
   properties {
    bIsEnabled=false
    visibility=Collapsed
    brush {
     bIsDynamicallyLoaded=false
     drawAs=Image
     imageType=NoImage
     mirroring=NoMirror
     resourceName=None
     resourceObject=&M_HUD_Thumbstick
     tiling=NoTile
     imageSize=<32 32>
     margin=<ltrb 0 0 0 0>
     outlineSettings {
      bUseBrushTransparency=false
      roundingType=HalfHeightRadius
      width=0
      color {colorUseRule=UseColor_Specified, specifiedColor=<rgba 0 0 0 0>}
      cornerRadii=<0 0 0 0>
     }
     tintColor {colorUseRule=UseColor_Specified, specifiedColor=<rgba 1 1 1 1>}
     uVRegion {bIsValid=false, max=<0 0>, min=<0 0>}
    }
   }
   slot {
    class=CanvasPanelSlot
    properties {layoutData {alignment=<0 0>, anchors {maximum=<0.3524 0.1557>, minimum=<0.3524 0.1557>}, offsets=<ltrb -20.535 -20.142 32 32>}}
   }
  }
  Gamepad_RightShoulder {
   widgetClassPath=&Image
   bIsVariable=true
   parent=CanvasPanel_Highlights
   widget=Gamepad_RightShoulder
   properties {
    bIsEnabled=false
    visibility=Collapsed
    brush {
     bIsDynamicallyLoaded=false
     drawAs=Image
     imageType=NoImage
     mirroring=NoMirror
     resourceName=None
     resourceObject=&M_HUD_Thumbstick
     tiling=NoTile
     imageSize=<32 32>
     margin=<ltrb 0 0 0 0>
     outlineSettings {
      bUseBrushTransparency=false
      roundingType=HalfHeightRadius
      width=0
      color {colorUseRule=UseColor_Specified, specifiedColor=<rgba 0 0 0 0>}
      cornerRadii=<0 0 0 0>
     }
     tintColor {colorUseRule=UseColor_Specified, specifiedColor=<rgba 1 1 1 1>}
     uVRegion {bIsValid=false, max=<0 0>, min=<0 0>}
    }
   }
   slot {
    class=CanvasPanelSlot
    properties {layoutData {alignment=<0 0>, anchors {maximum=<0.6482 0.2614>, minimum=<0.6482 0.2614>}, offsets=<ltrb -32.4845 -18.3336 64 32>}}
   }
  }
  Gamepad_RightTrigger {
   widgetClassPath=&Image
   bIsVariable=true
   parent=CanvasPanel_Highlights
   widget=Gamepad_RightTrigger
   properties {
    bIsEnabled=false
    visibility=Collapsed
    brush {
     bIsDynamicallyLoaded=false
     drawAs=Image
     imageType=NoImage
     mirroring=NoMirror
     resourceName=None
     resourceObject=&M_HUD_Thumbstick
     tiling=NoTile
     imageSize=<32 32>
     margin=<ltrb 0 0 0 0>
     outlineSettings {
      bUseBrushTransparency=false
      roundingType=HalfHeightRadius
      width=0
      color {colorUseRule=UseColor_Specified, specifiedColor=<rgba 0 0 0 0>}
      cornerRadii=<0 0 0 0>
     }
     tintColor {colorUseRule=UseColor_Specified, specifiedColor=<rgba 1 1 1 1>}
     uVRegion {bIsValid=false, max=<0 0>, min=<0 0>}
    }
   }
   slot {
    class=CanvasPanelSlot
    properties {layoutData {alignment=<0 0>, anchors {maximum=<0.6474 0.1524>, minimum=<0.6474 0.1524>}, offsets=<ltrb -15.0942 -16.6137 32 32>}}
   }
  }
 ]
}
