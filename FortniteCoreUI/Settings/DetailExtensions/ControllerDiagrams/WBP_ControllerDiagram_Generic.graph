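# Generic gamepad diagram widget for settings detail extensions, built on FortControllerHighlightDiagram to mark individual controller inputs.
# A generic controller image sits beneath a visibility switcher holding the inherited CanvasPanel_Highlights.
# That canvas anchors collapsed thumbstick-material highlight images over each face, D-pad and special button, stick, shoulder and trigger, named by gamepad input.
# The event graph is empty, so the Blueprint supplies only layout, with no highlight logic of its own.
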WidgetBlueprint /FortniteCoreUI/Settings/DetailExtensions/ControllerDiagrams/WBP_ControllerDiagram_Generic
graphs [EventGraph {nodes [
   Event PreConstruct()
   Event Construct()
   Event Tick()
  ]}]
variables []
widgets {
 info {inheritedWidgetCount=1, parentClass=/Script/FortniteUI.FortControllerHighlightDiagram, rootWidgetClass=&SizeBox, widgetCount=21}
 widgets [
  SizeBox_0 {widgetClassPath=&SizeBox, widget=SizeBox_0}
  Overlay_30 {widgetClassPath=&Overlay, parent=SizeBox_0, widget=Overlay_30, slot {class=SizeBoxSlot}}
  Image_ControllerInput {
   widgetClassPath=&Image
   bIsVariable=true
   parent=Overlay_30
   widget=Image_ControllerInput
   properties {
    bIsEnabled=false
    visibility=SelfHitTestInvisible
    brush {
     bIsDynamicallyLoaded=false
     drawAs=Image
     imageType=NoImage
     mirroring=NoMirror
     resourceName=None
     resourceObject=/FortUILibrary/Input/GamepadGeneric/Generic_ControllerConfig
     tiling=NoTile
     imageSize=<1024 512>
     margin=<ltrb 0 0 0 0>
     outlineSettings {
      bUseBrushTransparency=false
      roundingType=HalfHeightRadius
      width=0
      color {colorUseRule=UseColor_Specified, specifiedColor=<rgba 0 0 0 0>}
      cornerRadii=<0 0 0 0>
     }
     tintColor {colorUseRule=UseColor_Specified, specifiedColor=<rgba 1 1 1 1>}
     uVRegion {bIsValid=false, max=<0 0>, min=<0 0>}
    }
   }
   slot {class=OverlaySlot, properties {horizontalAlignment=HAlign_Fill, verticalAlignment=VAlign_Fill}}
  }
  Switcher_HighlightedElement {
   widgetClassPath=&CommonVisibilitySwitcher
   bIsVariable=true
   parent=Overlay_30
   widget=Switcher_HighlightedElement
   slot {class=OverlaySlot, properties {horizontalAlignment=HAlign_Fill, verticalAlignment=VAlign_Fill}}
  }
  CanvasPanel_Highlights {
   widgetClassPath=&CanvasPanel
   bInherited=true
   parent=Switcher_HighlightedElement
   widget=CanvasPanel_Highlights
   slot {class=CommonVisibilitySwitcherSlot}
  }
  Gamepad_FaceButton_Right {
   widgetClassPath=&Image
   bIsVariable=true
   parent=CanvasPanel_Highlights
   widget=Gamepad_FaceButton_Right
   properties {
    bIsEnabled=false
    visibility=Collapsed
    brush {
     bIsDynamicallyLoaded=false
     drawAs=Image
     imageType=NoImage
     mirroring=NoMirror
     resourceName=None
     resourceObject=/Game/Athena/HUD/Art/Materials/M_HUD_Thumbstick_Wide
     tiling=NoTile
     imageSize=<32 32>
     margin=<ltrb 0 0 0 0>
     outlineSettings {
      bUseBrushTransparency=false
      roundingType=HalfHeightRadius
      width=0
      color {colorUseRule=UseColor_Specified, specifiedColor=<rgba 0 0 0 0>}
      cornerRadii=<0 0 0 0>
     }
     tintColor {colorUseRule=UseColor_Specified, specifiedColor=<rgba 1 1 1 1>}
     uVRegion {bIsValid=false, max=<0 0>, min=<0 0>}
    }
   }
   slot {
    class=CanvasPanelSlot
    properties {layoutData {alignment=<0 0>, anchors {maximum=<0.7185 0.3402>, minimum=<0.7185 0.3402>}, offsets=<ltrb -24.9375 -24.9451 52 52>}}
   }
  }
  Gamepad_FaceButton_Left {
   widgetClassPath=&Image
   bIsVariable=true
   parent=CanvasPanel_Highlights
   widget=Gamepad_FaceButton_Left
   properties {
    bIsEnabled=false
    visibility=Collapsed
    brush {
     bIsDynamicallyLoaded=false
     drawAs=Image
     imageType=NoImage
     mirroring=NoMirror
     resourceName=None
     resourceObject=&M_HUD_Thumbstick_Wide
     tiling=NoTile
     imageSize=<64 64>
     margin=<ltrb 0 0 0 0>
     outlineSettings {
      bUseBrushTransparency=false
      roundingType=HalfHeightRadius
      width=0
      color {colorUseRule=UseColor_Specified, specifiedColor=<rgba 0 0 0 0>}
      cornerRadii=<0 0 0 0>
     }
     tintColor {colorUseRule=UseColor_Specified, specifiedColor=<rgba 1 1 1 1>}
     uVRegion {bIsValid=false, max=<0 0>, min=<0 0>}
    }
   }
   slot {
    class=CanvasPanelSlot
    properties {layoutData {alignment=<0 0>, anchors {maximum=<0.6224 0.3414>, minimum=<0.6224 0.3414>}, offsets=<ltrb -24.5541 -25.2217 52 52>}}
   }
  }
  Gamepad_FaceButton_Top {
   widgetClassPath=&Image
   bIsVariable=true
   parent=CanvasPanel_Highlights
   widget=Gamepad_FaceButton_Top
   properties {
    bIsEnabled=false
    visibility=Collapsed
    brush {
     bIsDynamicallyLoaded=false
     drawAs=Image
     imageType=NoImage
     mirroring=NoMirror
     resourceName=None
     resourceObject=&M_HUD_Thumbstick_Wide
     tiling=NoTile
     imageSize=<64 64>
     margin=<ltrb 0 0 0 0>
     outlineSettings {
      bUseBrushTransparency=false
      roundingType=HalfHeightRadius
      width=0
      color {colorUseRule=UseColor_Specified, specifiedColor=<rgba 0 0 0 0>}
      cornerRadii=<0 0 0 0>
     }
     tintColor {colorUseRule=UseColor_Specified, specifiedColor=<rgba 1 1 1 1>}
     uVRegion {bIsValid=false, max=<0 0>, min=<0 0>}
    }
   }
   slot {
    class=CanvasPanelSlot
    properties {layoutData {alignment=<0 0>, anchors {maximum=<0.6697 0.2432>, minimum=<0.6697 0.2432>}, offsets=<ltrb -25.7415 -23.6517 52 52>}}
   }
  }
  Gamepad_FaceButton_Bottom {
   widgetClassPath=&Image
   bIsVariable=true
   parent=CanvasPanel_Highlights
   widget=Gamepad_FaceButton_Bottom
   properties {
    bIsEnabled=false
    visibility=Collapsed
    brush {
     bIsDynamicallyLoaded=false
     drawAs=Image
     imageType=NoImage
     mirroring=NoMirror
     resourceName=None
     resourceObject=&M_HUD_Thumbstick_Wide
     tiling=NoTile
     imageSize=<64 64>
     margin=<ltrb 0 0 0 0>
     outlineSettings {
      bUseBrushTransparency=false
      roundingType=HalfHeightRadius
      width=0
      color {colorUseRule=UseColor_Specified, specifiedColor=<rgba 0 0 0 0>}
      cornerRadii=<0 0 0 0>
     }
     tintColor {colorUseRule=UseColor_Specified, specifiedColor=<rgba 1 1 1 1>}
     uVRegion {bIsValid=false, max=<0 0>, min=<0 0>}
    }
   }
   slot {
    class=CanvasPanelSlot
    properties {layoutData {alignment=<0 0>, anchors {maximum=<0.6711 0.4405>, minimum=<0.6711 0.4405>}, offsets=<ltrb -25.5139 -30.2215 52 52>}}
   }
  }
  Gamepad_Special_Right {
   widgetClassPath=&Image
   bIsVariable=true
   parent=CanvasPanel_Highlights
   widget=Gamepad_Special_Right
   properties {
    bIsEnabled=false
    visibility=Collapsed
    brush {
     bIsDynamicallyLoaded=false
     drawAs=Image
     imageType=NoImage
     mirroring=NoMirror
     resourceName=None
     resourceObject=/Game/Athena/HUD/Art/Materials/M_HUD_Thumbstick
     tiling=NoTile
     imageSize=<32 32>
     margin=<ltrb 0 0 0 0>
     outlineSettings {
      bUseBrushTransparency=false
      roundingType=HalfHeightRadius
      width=0
      color {colorUseRule=UseColor_Specified, specifiedColor=<rgba 0 0 0 0>}
      cornerRadii=<0 0 0 0>
     }
     tintColor {colorUseRule=UseColor_Specified, specifiedColor=<rgba 1 1 1 1>}
     uVRegion {bIsValid=false, max=<0 0>, min=<0 0>}
    }
   }
   slot {
    class=CanvasPanelSlot
    properties {layoutData {alignment=<0 0>, anchors {maximum=<0.5801 0.1701>, minimum=<0.5801 0.1701>}, offsets=<ltrb -16.7383 -15.1853 32 32>}}
   }
  }
  Gamepad_Special_Left {
   widgetClassPath=&Image
   bIsVariable=true
   parent=CanvasPanel_Highlights
   widget=Gamepad_Special_Left
   properties {
    bIsEnabled=false
    visibility=Collapsed
    brush {
     bIsDynamicallyLoaded=false
     drawAs=Image
     imageType=NoImage
     mirroring=NoMirror
     resourceName=None
     resourceObject=&M_HUD_Thumbstick
     tiling=NoTile
     imageSize=<32 32>
     margin=<ltrb 0 0 0 0>
     outlineSettings {
      bUseBrushTransparency=false
      roundingType=HalfHeightRadius
      width=0
      color {colorUseRule=UseColor_Specified, specifiedColor=<rgba 0 0 0 0>}
      cornerRadii=<0 0 0 0>
     }
     tintColor {colorUseRule=UseColor_Specified, specifiedColor=<rgba 1 1 1 1>}
     uVRegion {bIsValid=false, max=<0 0>, min=<0 0>}
    }
   }
   slot {
    class=CanvasPanelSlot
    properties {layoutData {alignment=<0 0>, anchors {maximum=<0.4201 0.1692>, minimum=<0.4201 0.1692>}, offsets=<ltrb -14.0555 -15.2587 32 32>}}
   }
  }
  Gamepad_LeftThumbstick {
   widgetClassPath=&Image
   bIsVariable=true
   parent=CanvasPanel_Highlights
   widget=Gamepad_LeftThumbstick
   properties {
    bIsEnabled=false
    visibility=Collapsed
    brush {
     bIsDynamicallyLoaded=false
     drawAs=Image
     imageType=NoImage
     mirroring=NoMirror
     resourceName=None
     resourceObject=&M_HUD_Thumbstick
     tiling=NoTile
     imageSize=<32 32>
     margin=<ltrb 0 0 0 0>
     outlineSettings {
      bUseBrushTransparency=false
      roundingType=HalfHeightRadius
      width=0
      color {colorUseRule=UseColor_Specified, specifiedColor=<rgba 0 0 0 0>}
      cornerRadii=<0 0 0 0>
     }
     tintColor {colorUseRule=UseColor_Specified, specifiedColor=<rgba 1 1 1 1>}
     uVRegion {bIsValid=false, max=<0 0>, min=<0 0>}
    }
   }
   slot {
    class=CanvasPanelSlot
    properties {layoutData {alignment=<0 0>, anchors {maximum=<0.4 0.6249>, minimum=<0.4 0.6249>}, offsets=<ltrb -26.5 -31.8887 60 60>}}
   }
  }
  Gamepad_RightThumbstick {
   widgetClassPath=&Image
   bIsVariable=true
   parent=CanvasPanel_Highlights
   widget=Gamepad_RightThumbstick
   properties {
    bIsEnabled=false
    visibility=Collapsed
    brush {
     bIsDynamicallyLoaded=false
     drawAs=Image
     imageType=NoImage
     mirroring=NoMirror
     resourceName=None
     resourceObject=&M_HUD_Thumbstick
     tiling=NoTile
     imageSize=<32 32>
     margin=<ltrb 0 0 0 0>
     outlineSettings {
      bUseBrushTransparency=false
      roundingType=HalfHeightRadius
      width=0
      color {colorUseRule=UseColor_Specified, specifiedColor=<rgba 0 0 0 0>}
      cornerRadii=<0 0 0 0>
     }
     tintColor {colorUseRule=UseColor_Specified, specifiedColor=<rgba 1 1 1 1>}
     uVRegion {bIsValid=false, max=<0 0>, min=<0 0>}
    }
   }
   slot {
    class=CanvasPanelSlot
    properties {layoutData {alignment=<0 0>, anchors {maximum=<0.6 0.62>, minimum=<0.6 0.62>}, offsets=<ltrb -32 -31.0546 60 60>}}
   }
  }
  Gamepad_DPad_Left {
   widgetClassPath=&Image
   bIsVariable=true
   parent=CanvasPanel_Highlights
   widget=Gamepad_DPad_Left
   properties {
    bIsEnabled=false
    visibility=Collapsed
    brush {
     bIsDynamicallyLoaded=false
     drawAs=Image
     imageType=NoImage
     mirroring=NoMirror
     resourceName=None
     resourceObject=&M_HUD_Thumbstick_Wide
     tiling=NoTile
     imageSize=<32 32>
     margin=<ltrb 0 0 0 0>
     outlineSettings {
      bUseBrushTransparency=false
      roundingType=HalfHeightRadius
      width=0
      color {colorUseRule=UseColor_Specified, specifiedColor=<rgba 0 0 0 0>}
      cornerRadii=<0 0 0 0>
     }
     tintColor {colorUseRule=UseColor_Specified, specifiedColor=<rgba 1 1 1 1>}
     uVRegion {bIsValid=false, max=<0 0>, min=<0 0>}
    }
   }
   slot {
    class=CanvasPanelSlot
    properties {layoutData {alignment=<0 0>, anchors {maximum=<0.2789 0.338>, minimum=<0.2789 0.338>}, offsets=<ltrb -24.0744 -26.5195 52 52>}}
   }
  }
  Gamepad_DPad_Right {
   widgetClassPath=&Image
   bIsVariable=true
   parent=CanvasPanel_Highlights
   widget=Gamepad_DPad_Right
   properties {
    bIsEnabled=false
    visibility=Collapsed
    brush {
     bIsDynamicallyLoaded=false
     drawAs=Image
     imageType=NoImage
     mirroring=NoMirror
     resourceName=None
     resourceObject=&M_HUD_Thumbstick_Wide
     tiling=NoTile
     imageSize=<32 32>
     margin=<ltrb 0 0 0 0>
     outlineSettings {
      bUseBrushTransparency=false
      roundingType=HalfHeightRadius
      width=0
      color {colorUseRule=UseColor_Specified, specifiedColor=<rgba 0 0 0 0>}
      cornerRadii=<0 0 0 0>
     }
     tintColor {colorUseRule=UseColor_Specified, specifiedColor=<rgba 1 1 1 1>}
     uVRegion {bIsValid=false, max=<0 0>, min=<0 0>}
    }
   }
   slot {
    class=CanvasPanelSlot
    properties {layoutData {alignment=<0 0>, anchors {maximum=<0.371 0.3407>, minimum=<0.371 0.3407>}, offsets=<ltrb -29.8527 -26.4063 52 52>}}
   }
  }
  Gamepad_DPad_Up {
   widgetClassPath=&Image
   bIsVariable=true
   parent=CanvasPanel_Highlights
   widget=Gamepad_DPad_Up
   properties {
    bIsEnabled=false
    visibility=Collapsed
    brush {
     bIsDynamicallyLoaded=false
     drawAs=Image
     imageType=NoImage
     mirroring=NoMirror
     resourceName=None
     resourceObject=&M_HUD_Thumbstick_Wide
     tiling=NoTile
     imageSize=<32 32>
     margin=<ltrb 0 0 0 0>
     outlineSettings {
      bUseBrushTransparency=false
      roundingType=HalfHeightRadius
      width=0
      color {colorUseRule=UseColor_Specified, specifiedColor=<rgba 0 0 0 0>}
      cornerRadii=<0 0 0 0>
     }
     tintColor {colorUseRule=UseColor_Specified, specifiedColor=<rgba 1 1 1 1>}
     uVRegion {bIsValid=false, max=<0 0>, min=<0 0>}
    }
   }
   slot {
    class=CanvasPanelSlot
    properties {layoutData {alignment=<0 0>, anchors {maximum=<0.3273 0.251>, minimum=<0.3273 0.251>}, offsets=<ltrb -25.3527 -37.0551 52 52>}}
   }
  }
  Gamepad_DPad_Down {
   widgetClassPath=&Image
   bIsVariable=true
   parent=CanvasPanel_Highlights
   widget=Gamepad_DPad_Down
   properties {
    bIsEnabled=false
    visibility=Collapsed
    brush {
     bIsDynamicallyLoaded=false
     drawAs=Image
     imageType=NoImage
     mirroring=NoMirror
     resourceName=None
     resourceObject=&M_HUD_Thumbstick_Wide
     tiling=NoTile
     imageSize=<32 32>
     margin=<ltrb 0 0 0 0>
     outlineSettings {
      bUseBrushTransparency=false
      roundingType=HalfHeightRadius
      width=0
      color {colorUseRule=UseColor_Specified, specifiedColor=<rgba 0 0 0 0>}
      cornerRadii=<0 0 0 0>
     }
     tintColor {colorUseRule=UseColor_Specified, specifiedColor=<rgba 1 1 1 1>}
     uVRegion {bIsValid=false, max=<0 0>, min=<0 0>}
    }
   }
   slot {
    class=CanvasPanelSlot
    properties {layoutData {alignment=<0 0>, anchors {maximum=<0.3242 0.4426>, minimum=<0.3242 0.4426>}, offsets=<ltrb -24.4448 -28.0004 52 52>}}
   }
  }
  Gamepad_RightShoulder {
   widgetClassPath=&Image
   bIsVariable=true
   parent=CanvasPanel_Highlights
   widget=Gamepad_RightShoulder
   properties {
    bIsEnabled=false
    visibility=Collapsed
    brush {
     bIsDynamicallyLoaded=false
     drawAs=Image
     imageType=NoImage
     mirroring=NoMirror
     resourceName=None
     resourceObject=&M_HUD_Thumbstick
     tiling=NoTile
     imageSize=<32 32>
     margin=<ltrb 0 0 0 0>
     outlineSettings {
      bUseBrushTransparency=false
      roundingType=HalfHeightRadius
      width=0
      color {colorUseRule=UseColor_Specified, specifiedColor=<rgba 0 0 0 0>}
      cornerRadii=<0 0 0 0>
     }
     tintColor {colorUseRule=UseColor_Specified, specifiedColor=<rgba 1 1 1 1>}
     uVRegion {bIsValid=false, max=<0 0>, min=<0 0>}
    }
   }
   slot {
    class=CanvasPanelSlot
    properties {layoutData {alignment=<0 0>, anchors {maximum=<0.7 0.1287>, minimum=<0.7 0.1287>}, offsets=<ltrb -37 -14.5003 64 32>}}
   }
  }
  Gamepad_RightTrigger {
   widgetClassPath=&Image
   bIsVariable=true
   parent=CanvasPanel_Highlights
   widget=Gamepad_RightTrigger
   properties {
    bIsEnabled=false
    visibility=Collapsed
    brush {
     bIsDynamicallyLoaded=false
     drawAs=Image
     imageType=NoImage
     mirroring=NoMirror
     resourceName=None
     resourceObject=&M_HUD_Thumbstick
     tiling=NoTile
     imageSize=<32 32>
     margin=<ltrb 0 0 0 0>
     outlineSettings {
      bUseBrushTransparency=false
      roundingType=HalfHeightRadius
      width=0
      color {colorUseRule=UseColor_Specified, specifiedColor=<rgba 0 0 0 0>}
      cornerRadii=<0 0 0 0>
     }
     tintColor {colorUseRule=UseColor_Specified, specifiedColor=<rgba 1 1 1 1>}
     uVRegion {bIsValid=false, max=<0 0>, min=<0 0>}
    }
   }
   slot {
    class=CanvasPanelSlot
    properties {layoutData {alignment=<0 0>, anchors {maximum=<0.6408 0.1>, minimum=<0.6408 0.1>}, offsets=<ltrb -14.9089 -15.5 32 32>}}
   }
  }
  Gamepad_LeftShoulder {
   widgetClassPath=&Image
   bIsVariable=true
   parent=CanvasPanel_Highlights
   widget=Gamepad_LeftShoulder
   properties {
    bIsEnabled=false
    visibility=Collapsed
    brush {
     bIsDynamicallyLoaded=false
     drawAs=Image
     imageType=NoImage
     mirroring=NoMirror
     resourceName=None
     resourceObject=&M_HUD_Thumbstick
     tiling=NoTile
     imageSize=<32 32>
     margin=<ltrb 0 0 0 0>
     outlineSettings {
      bUseBrushTransparency=false
      roundingType=HalfHeightRadius
      width=0
      color {colorUseRule=UseColor_Specified, specifiedColor=<rgba 0 0 0 0>}
      cornerRadii=<0 0 0 0>
     }
     tintColor {colorUseRule=UseColor_Specified, specifiedColor=<rgba 1 1 1 1>}
     uVRegion {bIsValid=false, max=<0 0>, min=<0 0>}
    }
   }
   slot {
    class=CanvasPanelSlot
    properties {layoutData {alignment=<0 0>, anchors {maximum=<0.3119 0.1292>, minimum=<0.3119 0.1292>}, offsets=<ltrb -34.3499 -17.5004 64 32>}}
   }
  }
  Gamepad_LeftTrigger {
   widgetClassPath=&Image
   bIsVariable=true
   parent=CanvasPanel_Highlights
   widget=Gamepad_LeftTrigger
   properties {
    bIsEnabled=false
    visibility=Collapsed
    brush {
     bIsDynamicallyLoaded=false
     drawAs=Image
     imageType=NoImage
     mirroring=NoMirror
     resourceName=None
     resourceObject=&M_HUD_Thumbstick
     tiling=NoTile
     imageSize=<32 32>
     margin=<ltrb 0 0 0 0>
     outlineSettings {
      bUseBrushTransparency=false
      roundingType=HalfHeightRadius
      width=0
      color {colorUseRule=UseColor_Specified, specifiedColor=<rgba 0 0 0 0>}
      cornerRadii=<0 0 0 0>
     }
     tintColor {colorUseRule=UseColor_Specified, specifiedColor=<rgba 1 1 1 1>}
     uVRegion {bIsValid=false, max=<0 0>, min=<0 0>}
    }
   }
   slot {
    class=CanvasPanelSlot
    properties {layoutData {alignment=<0 0>, anchors {maximum=<0.3575 0.1>, minimum=<0.3575 0.1>}, offsets=<ltrb -15.3328 -13.5 32 32>}}
   }
  }
 ]
}
