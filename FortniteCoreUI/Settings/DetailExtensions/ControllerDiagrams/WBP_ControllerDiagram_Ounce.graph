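# Controller diagram widget for the settings detail panel, showing the Switch 2 (Ounce) gamepad for input highlighting.
# A controller config image sits beneath a visibility switcher holding the inherited highlight canvas.
# The canvas anchors collapsed, named thumbstick-material images over each button, D-pad direction, stick, shoulder and trigger.
# The event graph holds only unconnected events, so highlight behavior is left to the FortControllerHighlightDiagram parent class.
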WidgetBlueprint /FortniteCoreUI/Settings/DetailExtensions/ControllerDiagrams/WBP_ControllerDiagram_Ounce
graphs [EventGraph {nodes [
   Event PreConstruct()
   Event Construct()
   Event Tick()
  ]}]
variables []
widgets {
 info {inheritedWidgetCount=1, parentClass=/Script/FortniteUI.FortControllerHighlightDiagram, rootWidgetClass=&SizeBox, widgetCount=21}
 widgets [
  SizeBox_0 {widgetClassPath=&SizeBox, widget=SizeBox_0}
  Overlay_30 {widgetClassPath=&Overlay, parent=SizeBox_0, widget=Overlay_30, slot {class=SizeBoxSlot}}
  Image_ControllerInput {
   widgetClassPath=&Image
   bIsVariable=true
   parent=Overlay_30
   widget=Image_ControllerInput
   properties {
    bIsEnabled=false
    visibility=SelfHitTestInvisible
    brush {
     bIsDynamicallyLoaded=false
     drawAs=Image
     imageType=FullColor
     mirroring=NoMirror
     resourceName=None
     resourceObject=/FortUILibrary/Input/GamepadSwitch2/T_UI_Ounce_ControllerConfig
     tiling=NoTile
     imageSize=<1024 672>
     margin=<ltrb 0 0 0 0>
     outlineSettings {
      bUseBrushTransparency=false
      roundingType=HalfHeightRadius
      width=0
      color {colorUseRule=UseColor_Specified, specifiedColor=<rgba 0 0 0 0>}
      cornerRadii=<0 0 0 0>
     }
     tintColor {colorUseRule=UseColor_Specified, specifiedColor=<rgba 1 1 1 1>}
     uVRegion {bIsValid=false, max=<0 0>, min=<0 0>}
    }
   }
   slot {class=OverlaySlot, properties {horizontalAlignment=HAlign_Fill, verticalAlignment=VAlign_Fill}}
  }
  Switcher_HighlightedElement {
   widgetClassPath=&CommonVisibilitySwitcher
   bIsVariable=true
   parent=Overlay_30
   widget=Switcher_HighlightedElement
   slot {class=OverlaySlot, properties {horizontalAlignment=HAlign_Fill, verticalAlignment=VAlign_Fill}}
  }
  CanvasPanel_Highlights {
   widgetClassPath=&CanvasPanel
   bInherited=true
   parent=Switcher_HighlightedElement
   widget=CanvasPanel_Highlights
   slot {class=CommonVisibilitySwitcherSlot}
  }
  Gamepad_FaceButton_Right {
   widgetClassPath=&Image
   bIsVariable=true
   parent=CanvasPanel_Highlights
   widget=Gamepad_FaceButton_Right
   properties {
    bIsEnabled=false
    visibility=Collapsed
    brush {
     bIsDynamicallyLoaded=false
     drawAs=Image
     imageType=NoImage
     mirroring=NoMirror
     resourceName=None
     resourceObject=/Game/Athena/HUD/Art/Materials/M_HUD_Thumbstick_Wide
     tiling=NoTile
     imageSize=<32 32>
     margin=<ltrb 0 0 0 0>
     outlineSettings {
      bUseBrushTransparency=false
      roundingType=HalfHeightRadius
      width=0
      color {colorUseRule=UseColor_Specified, specifiedColor=<rgba 0 0 0 0>}
      cornerRadii=<0 0 0 0>
     }
     tintColor {colorUseRule=UseColor_Specified, specifiedColor=<rgba 1 1 1 1>}
     uVRegion {bIsValid=false, max=<0 0>, min=<0 0>}
    }
   }
   slot {
    class=CanvasPanelSlot
    properties {layoutData {alignment=<0.5 0.5>, anchors {maximum=<0.67 0.4155>, minimum=<0.67 0.4155>}, offsets=<ltrb 0 0 52 52>}}
   }
  }
  Gamepad_FaceButton_Left {
   widgetClassPath=&Image
   bIsVariable=true
   parent=CanvasPanel_Highlights
   widget=Gamepad_FaceButton_Left
   properties {
    bIsEnabled=false
    visibility=Collapsed
    brush {
     bIsDynamicallyLoaded=false
     drawAs=Image
     imageType=NoImage
     mirroring=NoMirror
     resourceName=None
     resourceObject=&M_HUD_Thumbstick_Wide
     tiling=NoTile
     imageSize=<64 64>
     margin=<ltrb 0 0 0 0>
     outlineSettings {
      bUseBrushTransparency=false
      roundingType=HalfHeightRadius
      width=0
      color {colorUseRule=UseColor_Specified, specifiedColor=<rgba 0 0 0 0>}
      cornerRadii=<0 0 0 0>
     }
     tintColor {colorUseRule=UseColor_Specified, specifiedColor=<rgba 1 1 1 1>}
     uVRegion {bIsValid=false, max=<0 0>, min=<0 0>}
    }
   }
   slot {
    class=CanvasPanelSlot
    properties {layoutData {alignment=<0.5 0.5>, anchors {maximum=<0.6109 0.4152>, minimum=<0.6109 0.4152>}, offsets=<ltrb 0 0 52 52>}}
   }
  }
  Gamepad_FaceButton_Top {
   widgetClassPath=&Image
   bIsVariable=true
   parent=CanvasPanel_Highlights
   widget=Gamepad_FaceButton_Top
   properties {
    bIsEnabled=false
    visibility=Collapsed
    brush {
     bIsDynamicallyLoaded=false
     drawAs=Image
     imageType=NoImage
     mirroring=NoMirror
     resourceName=None
     resourceObject=&M_HUD_Thumbstick_Wide
     tiling=NoTile
     imageSize=<64 64>
     margin=<ltrb 0 0 0 0>
     outlineSettings {
      bUseBrushTransparency=false
      roundingType=HalfHeightRadius
      width=0
      color {colorUseRule=UseColor_Specified, specifiedColor=<rgba 0 0 0 0>}
      cornerRadii=<0 0 0 0>
     }
     tintColor {colorUseRule=UseColor_Specified, specifiedColor=<rgba 1 1 1 1>}
     uVRegion {bIsValid=false, max=<0 0>, min=<0 0>}
    }
   }
   slot {
    class=CanvasPanelSlot
    properties {layoutData {alignment=<0.5 0.5>, anchors {maximum=<0.6406 0.3717>, minimum=<0.6406 0.3717>}, offsets=<ltrb 0 0 52 52>}}
   }
  }
  Gamepad_FaceButton_Bottom {
   widgetClassPath=&Image
   bIsVariable=true
   parent=CanvasPanel_Highlights
   widget=Gamepad_FaceButton_Bottom
   properties {
    bIsEnabled=false
    visibility=Collapsed
    brush {
     bIsDynamicallyLoaded=false
     drawAs=Image
     imageType=NoImage
     mirroring=NoMirror
     resourceName=None
     resourceObject=&M_HUD_Thumbstick_Wide
     tiling=NoTile
     imageSize=<64 64>
     margin=<ltrb 0 0 0 0>
     outlineSettings {
      bUseBrushTransparency=false
      roundingType=HalfHeightRadius
      width=0
      color {colorUseRule=UseColor_Specified, specifiedColor=<rgba 0 0 0 0>}
      cornerRadii=<0 0 0 0>
     }
     tintColor {colorUseRule=UseColor_Specified, specifiedColor=<rgba 1 1 1 1>}
     uVRegion {bIsValid=false, max=<0 0>, min=<0 0>}
    }
   }
   slot {
    class=CanvasPanelSlot
    properties {layoutData {alignment=<0.5 0.5>, anchors {maximum=<0.6407 0.4583>, minimum=<0.6407 0.4583>}, offsets=<ltrb 0 0 52 52>}}
   }
  }
  Gamepad_Special_Right {
   widgetClassPath=&Image
   bIsVariable=true
   parent=CanvasPanel_Highlights
   widget=Gamepad_Special_Right
   properties {
    bIsEnabled=false
    visibility=Collapsed
    brush {
     bIsDynamicallyLoaded=false
     drawAs=Image
     imageType=NoImage
     mirroring=NoMirror
     resourceName=None
     resourceObject=/Game/Athena/HUD/Art/Materials/M_HUD_Thumbstick
     tiling=NoTile
     imageSize=<32 32>
     margin=<ltrb 0 0 0 0>
     outlineSettings {
      bUseBrushTransparency=false
      roundingType=HalfHeightRadius
      width=0
      color {colorUseRule=UseColor_Specified, specifiedColor=<rgba 0 0 0 0>}
      cornerRadii=<0 0 0 0>
     }
     tintColor {colorUseRule=UseColor_Specified, specifiedColor=<rgba 1 1 1 1>}
     uVRegion {bIsValid=false, max=<0 0>, min=<0 0>}
    }
   }
   slot {
    class=CanvasPanelSlot
    properties {layoutData {alignment=<0.5 0.5>, anchors {maximum=<0.6041 0.3298>, minimum=<0.6041 0.3298>}, offsets=<ltrb 0 0 32 32>}}
   }
  }
  Gamepad_Special_Left {
   widgetClassPath=&Image
   bIsVariable=true
   parent=CanvasPanel_Highlights
   widget=Gamepad_Special_Left
   properties {
    bIsEnabled=false
    visibility=Collapsed
    brush {
     bIsDynamicallyLoaded=false
     drawAs=Image
     imageType=NoImage
     mirroring=NoMirror
     resourceName=None
     resourceObject=&M_HUD_Thumbstick
     tiling=NoTile
     imageSize=<32 32>
     margin=<ltrb 0 0 0 0>
     outlineSettings {
      bUseBrushTransparency=false
      roundingType=HalfHeightRadius
      width=0
      color {colorUseRule=UseColor_Specified, specifiedColor=<rgba 0 0 0 0>}
      cornerRadii=<0 0 0 0>
     }
     tintColor {colorUseRule=UseColor_Specified, specifiedColor=<rgba 1 1 1 1>}
     uVRegion {bIsValid=false, max=<0 0>, min=<0 0>}
    }
   }
   slot {
    class=CanvasPanelSlot
    properties {layoutData {alignment=<0.5 0.5>, anchors {maximum=<0.4 0.3295>, minimum=<0.4 0.3295>}, offsets=<ltrb 0 0 32 32>}}
   }
  }
  Gamepad_LeftThumbstick {
   widgetClassPath=&Image
   bIsVariable=true
   parent=CanvasPanel_Highlights
   widget=Gamepad_LeftThumbstick
   properties {
    bIsEnabled=false
    visibility=Collapsed
    brush {
     bIsDynamicallyLoaded=false
     drawAs=Image
     imageType=NoImage
     mirroring=NoMirror
     resourceName=None
     resourceObject=&M_HUD_Thumbstick
     tiling=NoTile
     imageSize=<32 32>
     margin=<ltrb 0 0 0 0>
     outlineSettings {
      bUseBrushTransparency=false
      roundingType=HalfHeightRadius
      width=0
      color {colorUseRule=UseColor_Specified, specifiedColor=<rgba 0 0 0 0>}
      cornerRadii=<0 0 0 0>
     }
     tintColor {colorUseRule=UseColor_Specified, specifiedColor=<rgba 1 1 1 1>}
     uVRegion {bIsValid=false, max=<0 0>, min=<0 0>}
    }
   }
   slot {
    class=CanvasPanelSlot
    properties {layoutData {alignment=<0.5 0.5>, anchors {maximum=<0.3639 0.4153>, minimum=<0.3639 0.4153>}, offsets=<ltrb 0 0 60 60>}}
   }
  }
  Gamepad_RightThumbstick {
   widgetClassPath=&Image
   bIsVariable=true
   parent=CanvasPanel_Highlights
   widget=Gamepad_RightThumbstick
   properties {
    bIsEnabled=false
    visibility=Collapsed
    brush {
     bIsDynamicallyLoaded=false
     drawAs=Image
     imageType=NoImage
     mirroring=NoMirror
     resourceName=None
     resourceObject=&M_HUD_Thumbstick
     tiling=NoTile
     imageSize=<32 32>
     margin=<ltrb 0 0 0 0>
     outlineSettings {
      bUseBrushTransparency=false
      roundingType=HalfHeightRadius
      width=0
      color {colorUseRule=UseColor_Specified, specifiedColor=<rgba 0 0 0 0>}
      cornerRadii=<0 0 0 0>
     }
     tintColor {colorUseRule=UseColor_Specified, specifiedColor=<rgba 1 1 1 1>}
     uVRegion {bIsValid=false, max=<0 0>, min=<0 0>}
    }
   }
   slot {
    class=CanvasPanelSlot
    properties {layoutData {alignment=<0.5 0.5>, anchors {maximum=<0.6407 0.57>, minimum=<0.6407 0.57>}, offsets=<ltrb 0 0 60 60>}}
   }
  }
  Gamepad_DPad_Left {
   widgetClassPath=&Image
   bIsVariable=true
   parent=CanvasPanel_Highlights
   widget=Gamepad_DPad_Left
   properties {
    bIsEnabled=false
    visibility=Collapsed
    brush {
     bIsDynamicallyLoaded=false
     drawAs=Image
     imageType=NoImage
     mirroring=NoMirror
     resourceName=None
     resourceObject=&M_HUD_Thumbstick_Wide
     tiling=NoTile
     imageSize=<32 32>
     margin=<ltrb 0 0 0 0>
     outlineSettings {
      bUseBrushTransparency=false
      roundingType=HalfHeightRadius
      width=0
      color {colorUseRule=UseColor_Specified, specifiedColor=<rgba 0 0 0 0>}
      cornerRadii=<0 0 0 0>
     }
     tintColor {colorUseRule=UseColor_Specified, specifiedColor=<rgba 1 1 1 1>}
     uVRegion {bIsValid=false, max=<0 0>, min=<0 0>}
    }
   }
   slot {
    class=CanvasPanelSlot
    properties {layoutData {alignment=<0.5 0.5>, anchors {maximum=<0.3343 0.5699>, minimum=<0.3343 0.5699>}, offsets=<ltrb 0 0 52 52>}}
   }
  }
  Gamepad_DPad_Right {
   widgetClassPath=&Image
   bIsVariable=true
   parent=CanvasPanel_Highlights
   widget=Gamepad_DPad_Right
   properties {
    bIsEnabled=false
    visibility=Collapsed
    brush {
     bIsDynamicallyLoaded=false
     drawAs=Image
     imageType=NoImage
     mirroring=NoMirror
     resourceName=None
     resourceObject=&M_HUD_Thumbstick_Wide
     tiling=NoTile
     imageSize=<32 32>
     margin=<ltrb 0 0 0 0>
     outlineSettings {
      bUseBrushTransparency=false
      roundingType=HalfHeightRadius
      width=0
      color {colorUseRule=UseColor_Specified, specifiedColor=<rgba 0 0 0 0>}
      cornerRadii=<0 0 0 0>
     }
     tintColor {colorUseRule=UseColor_Specified, specifiedColor=<rgba 1 1 1 1>}
     uVRegion {bIsValid=false, max=<0 0>, min=<0 0>}
    }
   }
   slot {
    class=CanvasPanelSlot
    properties {layoutData {alignment=<0.5 0.5>, anchors {maximum=<0.3936 0.5704>, minimum=<0.3936 0.5704>}, offsets=<ltrb 0 0 52 52>}}
   }
  }
  Gamepad_DPad_Up {
   widgetClassPath=&Image
   bIsVariable=true
   parent=CanvasPanel_Highlights
   widget=Gamepad_DPad_Up
   properties {
    bIsEnabled=false
    visibility=Collapsed
    brush {
     bIsDynamicallyLoaded=false
     drawAs=Image
     imageType=NoImage
     mirroring=NoMirror
     resourceName=None
     resourceObject=&M_HUD_Thumbstick_Wide
     tiling=NoTile
     imageSize=<32 32>
     margin=<ltrb 0 0 0 0>
     outlineSettings {
      bUseBrushTransparency=false
      roundingType=HalfHeightRadius
      width=0
      color {colorUseRule=UseColor_Specified, specifiedColor=<rgba 0 0 0 0>}
      cornerRadii=<0 0 0 0>
     }
     tintColor {colorUseRule=UseColor_Specified, specifiedColor=<rgba 1 1 1 1>}
     uVRegion {bIsValid=false, max=<0 0>, min=<0 0>}
    }
   }
   slot {
    class=CanvasPanelSlot
    properties {layoutData {alignment=<0.5 0.5>, anchors {maximum=<0.3637 0.5272>, minimum=<0.3637 0.5272>}, offsets=<ltrb 0 0 52 52>}}
   }
  }
  Gamepad_DPad_Down {
   widgetClassPath=&Image
   bIsVariable=true
   parent=CanvasPanel_Highlights
   widget=Gamepad_DPad_Down
   properties {
    bIsEnabled=false
    visibility=Collapsed
    brush {
     bIsDynamicallyLoaded=false
     drawAs=Image
     imageType=NoImage
     mirroring=NoMirror
     resourceName=None
     resourceObject=&M_HUD_Thumbstick_Wide
     tiling=NoTile
     imageSize=<32 32>
     margin=<ltrb 0 0 0 0>
     outlineSettings {
      bUseBrushTransparency=false
      roundingType=HalfHeightRadius
      width=0
      color {colorUseRule=UseColor_Specified, specifiedColor=<rgba 0 0 0 0>}
      cornerRadii=<0 0 0 0>
     }
     tintColor {colorUseRule=UseColor_Specified, specifiedColor=<rgba 1 1 1 1>}
     uVRegion {bIsValid=false, max=<0 0>, min=<0 0>}
    }
   }
   slot {
    class=CanvasPanelSlot
    properties {layoutData {alignment=<0.5 0.5>, anchors {maximum=<0.3638 0.613>, minimum=<0.3638 0.613>}, offsets=<ltrb 0 0 52 52>}}
   }
  }
  Gamepad_LeftShoulder {
   widgetClassPath=&Image
   bIsVariable=true
   parent=CanvasPanel_Highlights
   widget=Gamepad_LeftShoulder
   properties {
    bIsEnabled=false
    visibility=Collapsed
    brush {
     bIsDynamicallyLoaded=false
     drawAs=Image
     imageType=NoImage
     mirroring=NoMirror
     resourceName=None
     resourceObject=&M_HUD_Thumbstick
     tiling=NoTile
     imageSize=<32 32>
     margin=<ltrb 0 0 0 0>
     outlineSettings {
      bUseBrushTransparency=false
      roundingType=HalfHeightRadius
      width=0
      color {colorUseRule=UseColor_Specified, specifiedColor=<rgba 0 0 0 0>}
      cornerRadii=<0 0 0 0>
     }
     tintColor {colorUseRule=UseColor_Specified, specifiedColor=<rgba 1 1 1 1>}
     uVRegion {bIsValid=false, max=<0 0>, min=<0 0>}
    }
   }
   slot {
    class=CanvasPanelSlot
    properties {layoutData {alignment=<0.5 0.5>, anchors {maximum=<0.3534 0.2779>, minimum=<0.3534 0.2779>}, offsets=<ltrb 0 0 64 32>}}
   }
  }
  Gamepad_LeftTrigger {
   widgetClassPath=&Image
   bIsVariable=true
   parent=CanvasPanel_Highlights
   widget=Gamepad_LeftTrigger
   properties {
    bIsEnabled=false
    visibility=Collapsed
    brush {
     bIsDynamicallyLoaded=false
     drawAs=Image
     imageType=NoImage
     mirroring=NoMirror
     resourceName=None
     resourceObject=&M_HUD_Thumbstick
     tiling=NoTile
     imageSize=<32 32>
     margin=<ltrb 0 0 0 0>
     outlineSettings {
      bUseBrushTransparency=false
      roundingType=HalfHeightRadius
      width=0
      color {colorUseRule=UseColor_Specified, specifiedColor=<rgba 0 0 0 0>}
      cornerRadii=<0 0 0 0>
     }
     tintColor {colorUseRule=UseColor_Specified, specifiedColor=<rgba 1 1 1 1>}
     uVRegion {bIsValid=false, max=<0 0>, min=<0 0>}
    }
   }
   slot {
    class=CanvasPanelSlot
    properties {layoutData {alignment=<0.5 0.5>, anchors {maximum=<0.3531 0.1549>, minimum=<0.3531 0.1549>}, offsets=<ltrb 0 0 32 32>}}
   }
  }
  Gamepad_RightShoulder {
   widgetClassPath=&Image
   bIsVariable=true
   parent=CanvasPanel_Highlights
   widget=Gamepad_RightShoulder
   properties {
    bIsEnabled=false
    visibility=Collapsed
    brush {
     bIsDynamicallyLoaded=false
     drawAs=Image
     imageType=NoImage
     mirroring=NoMirror
     resourceName=None
     resourceObject=&M_HUD_Thumbstick
     tiling=NoTile
     imageSize=<32 32>
     margin=<ltrb 0 0 0 0>
     outlineSettings {
      bUseBrushTransparency=false
      roundingType=HalfHeightRadius
      width=0
      color {colorUseRule=UseColor_Specified, specifiedColor=<rgba 0 0 0 0>}
      cornerRadii=<0 0 0 0>
     }
     tintColor {colorUseRule=UseColor_Specified, specifiedColor=<rgba 1 1 1 1>}
     uVRegion {bIsValid=false, max=<0 0>, min=<0 0>}
    }
   }
   slot {
    class=CanvasPanelSlot
    properties {layoutData {alignment=<0.5 0.5>, anchors {maximum=<0.6492 0.2744>, minimum=<0.6492 0.2744>}, offsets=<ltrb 0 0 64 32>}}
   }
  }
  Gamepad_RightTrigger {
   widgetClassPath=&Image
   bIsVariable=true
   parent=CanvasPanel_Highlights
   widget=Gamepad_RightTrigger
   properties {
    bIsEnabled=false
    visibility=Collapsed
    brush {
     bIsDynamicallyLoaded=false
     drawAs=Image
     imageType=NoImage
     mirroring=NoMirror
     resourceName=None
     resourceObject=&M_HUD_Thumbstick
     tiling=NoTile
     imageSize=<32 32>
     margin=<ltrb 0 0 0 0>
     outlineSettings {
      bUseBrushTransparency=false
      roundingType=HalfHeightRadius
      width=0
      color {colorUseRule=UseColor_Specified, specifiedColor=<rgba 0 0 0 0>}
      cornerRadii=<0 0 0 0>
     }
     tintColor {colorUseRule=UseColor_Specified, specifiedColor=<rgba 1 1 1 1>}
     uVRegion {bIsValid=false, max=<0 0>, min=<0 0>}
    }
   }
   slot {
    class=CanvasPanelSlot
    properties {layoutData {alignment=<0.5 0.5>, anchors {maximum=<0.6487 0.1546>, minimum=<0.6487 0.1546>}, offsets=<ltrb 0 0 32 32>}}
   }
  }
 ]
}
