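# PS5 controller diagram widget for settings detail extensions, used to highlight individual gamepad inputs over a DualSense layout image.
# It derives from FortControllerHighlightDiagram and layers the controller image beneath a visibility switcher holding the inherited highlight canvas.
# Collapsed, anchor-positioned images named per input (face buttons, D-pad, sticks, shoulders, triggers, specials) use thumbstick HUD materials as markers.
# The event graph holds only empty stub events, so no highlight logic lives here.
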WidgetBlueprint /FortniteCoreUI/Settings/DetailExtensions/ControllerDiagrams/WBP_ControllerDiagram_PS5
graphs [EventGraph {nodes [
   Event PreConstruct()
   Event Construct()
   Event Tick()
  ]}]
variables []
widgets {
 info {inheritedWidgetCount=1, parentClass=/Script/FortniteUI.FortControllerHighlightDiagram, rootWidgetClass=&SizeBox, widgetCount=21}
 widgets [
  SizeBox_0 {widgetClassPath=&SizeBox, widget=SizeBox_0}
  Overlay_30 {widgetClassPath=&Overlay, parent=SizeBox_0, widget=Overlay_30, slot {class=SizeBoxSlot}}
  Image_ControllerInput {
   widgetClassPath=&Image
   bIsVariable=true
   parent=Overlay_30
   widget=Image_ControllerInput
   properties {
    bIsEnabled=false
    visibility=SelfHitTestInvisible
    brush {
     bIsDynamicallyLoaded=false
     drawAs=Image
     imageType=NoImage
     mirroring=NoMirror
     resourceName=None
     resourceObject=/FortUILibrary/Input/GamepadPS5/PS5_ControllerConfig
     tiling=NoTile
     imageSize=<1028 512>
     margin=<ltrb 0 0 0 0>
     outlineSettings {
      bUseBrushTransparency=false
      roundingType=HalfHeightRadius
      width=0
      color {colorUseRule=UseColor_Specified, specifiedColor=<rgba 0 0 0 0>}
      cornerRadii=<0 0 0 0>
     }
     tintColor {colorUseRule=UseColor_Specified, specifiedColor=<rgba 1 1 1 1>}
     uVRegion {bIsValid=false, max=<0 0>, min=<0 0>}
    }
   }
   slot {class=OverlaySlot, properties {horizontalAlignment=HAlign_Fill, verticalAlignment=VAlign_Fill}}
  }
  Switcher_HighlightedElement {
   widgetClassPath=&CommonVisibilitySwitcher
   bIsVariable=true
   parent=Overlay_30
   widget=Switcher_HighlightedElement
   slot {class=OverlaySlot, properties {horizontalAlignment=HAlign_Fill, verticalAlignment=VAlign_Fill}}
  }
  CanvasPanel_Highlights {
   widgetClassPath=&CanvasPanel
   bInherited=true
   parent=Switcher_HighlightedElement
   widget=CanvasPanel_Highlights
   slot {class=CommonVisibilitySwitcherSlot}
  }
  Gamepad_FaceButton_Right {
   widgetClassPath=&Image
   bIsVariable=true
   parent=CanvasPanel_Highlights
   widget=Gamepad_FaceButton_Right
   properties {
    bIsEnabled=false
    visibility=Collapsed
    brush {
     bIsDynamicallyLoaded=false
     drawAs=Image
     imageType=NoImage
     mirroring=NoMirror
     resourceName=None
     resourceObject=/Game/Athena/HUD/Art/Materials/M_HUD_Thumbstick_Wide
     tiling=NoTile
     imageSize=<32 32>
     margin=<ltrb 0 0 0 0>
     outlineSettings {
      bUseBrushTransparency=false
      roundingType=HalfHeightRadius
      width=0
      color {colorUseRule=UseColor_Specified, specifiedColor=<rgba 0 0 0 0>}
      cornerRadii=<0 0 0 0>
     }
     tintColor {colorUseRule=UseColor_Specified, specifiedColor=<rgba 1 1 1 1>}
     uVRegion {bIsValid=false, max=<0 0>, min=<0 0>}
    }
   }
   slot {
    class=CanvasPanelSlot
    properties {layoutData {alignment=<0 0>, anchors {maximum=<0.7838 0.5226>, minimum=<0.7838 0.5226>}, offsets=<ltrb -28.9382 -24.4447 52 52>}}
   }
  }
  Gamepad_FaceButton_Left {
   widgetClassPath=&Image
   bIsVariable=true
   parent=CanvasPanel_Highlights
   widget=Gamepad_FaceButton_Left
   properties {
    bIsEnabled=false
    visibility=Collapsed
    brush {
     bIsDynamicallyLoaded=false
     drawAs=Image
     imageType=NoImage
     mirroring=NoMirror
     resourceName=None
     resourceObject=&M_HUD_Thumbstick_Wide
     tiling=NoTile
     imageSize=<64 64>
     margin=<ltrb 0 0 0 0>
     outlineSettings {
      bUseBrushTransparency=false
      roundingType=HalfHeightRadius
      width=0
      color {colorUseRule=UseColor_Specified, specifiedColor=<rgba 0 0 0 0>}
      cornerRadii=<0 0 0 0>
     }
     tintColor {colorUseRule=UseColor_Specified, specifiedColor=<rgba 1 1 1 1>}
     uVRegion {bIsValid=false, max=<0 0>, min=<0 0>}
    }
   }
   slot {
    class=CanvasPanelSlot
    properties {layoutData {alignment=<0 0>, anchors {maximum=<0.6755 0.5267>, minimum=<0.6755 0.5267>}, offsets=<ltrb -25.0555 -24.8884 52 52>}}
   }
  }
  Gamepad_FaceButton_Top {
   widgetClassPath=&Image
   bIsVariable=true
   parent=CanvasPanel_Highlights
   widget=Gamepad_FaceButton_Top
   properties {
    bIsEnabled=false
    visibility=Collapsed
    brush {
     bIsDynamicallyLoaded=false
     drawAs=Image
     imageType=NoImage
     mirroring=NoMirror
     resourceName=None
     resourceObject=&M_HUD_Thumbstick_Wide
     tiling=NoTile
     imageSize=<64 64>
     margin=<ltrb 0 0 0 0>
     outlineSettings {
      bUseBrushTransparency=false
      roundingType=HalfHeightRadius
      width=0
      color {colorUseRule=UseColor_Specified, specifiedColor=<rgba 0 0 0 0>}
      cornerRadii=<0 0 0 0>
     }
     tintColor {colorUseRule=UseColor_Specified, specifiedColor=<rgba 1 1 1 1>}
     uVRegion {bIsValid=false, max=<0 0>, min=<0 0>}
    }
   }
   slot {
    class=CanvasPanelSlot
    properties {layoutData {alignment=<0 0>, anchors {maximum=<0.7316 0.4356>, minimum=<0.7316 0.4356>}, offsets=<ltrb -24.7411 -30.4858 52 52>}}
   }
  }
  Gamepad_FaceButton_Bottom {
   widgetClassPath=&Image
   bIsVariable=true
   parent=CanvasPanel_Highlights
   widget=Gamepad_FaceButton_Bottom
   properties {
    bIsEnabled=false
    visibility=Collapsed
    brush {
     bIsDynamicallyLoaded=false
     drawAs=Image
     imageType=NoImage
     mirroring=NoMirror
     resourceName=None
     resourceObject=&M_HUD_Thumbstick_Wide
     tiling=NoTile
     imageSize=<64 64>
     margin=<ltrb 0 0 0 0>
     outlineSettings {
      bUseBrushTransparency=false
      roundingType=HalfHeightRadius
      width=0
      color {colorUseRule=UseColor_Specified, specifiedColor=<rgba 0 0 0 0>}
      cornerRadii=<0 0 0 0>
     }
     tintColor {colorUseRule=UseColor_Specified, specifiedColor=<rgba 1 1 1 1>}
     uVRegion {bIsValid=false, max=<0 0>, min=<0 0>}
    }
   }
   slot {
    class=CanvasPanelSlot
    properties {layoutData {alignment=<0 0>, anchors {maximum=<0.7278 0.6144>, minimum=<0.7278 0.6144>}, offsets=<ltrb -25.3472 -23.5552 52 52>}}
   }
  }
  Gamepad_Special_Right {
   widgetClassPath=&Image
   bIsVariable=true
   parent=CanvasPanel_Highlights
   widget=Gamepad_Special_Right
   properties {
    bIsEnabled=false
    visibility=Collapsed
    brush {
     bIsDynamicallyLoaded=false
     drawAs=Image
     imageType=NoImage
     mirroring=NoMirror
     resourceName=None
     resourceObject=/Game/Athena/HUD/Art/Materials/M_HUD_Thumbstick
     tiling=NoTile
     imageSize=<32 32>
     margin=<ltrb 0 0 0 0>
     outlineSettings {
      bUseBrushTransparency=false
      roundingType=HalfHeightRadius
      width=0
      color {colorUseRule=UseColor_Specified, specifiedColor=<rgba 0 0 0 0>}
      cornerRadii=<0 0 0 0>
     }
     tintColor {colorUseRule=UseColor_Specified, specifiedColor=<rgba 1 1 1 1>}
     uVRegion {bIsValid=false, max=<0 0>, min=<0 0>}
    }
   }
   slot {
    class=CanvasPanelSlot
    properties {layoutData {alignment=<0 0>, anchors {maximum=<0.6733 0.3838>, minimum=<0.6733 0.3838>}, offsets=<ltrb -12.7399 -26.5191 32 52>}}
   }
  }
  Gamepad_Special_Left {
   widgetClassPath=&Image
   bIsVariable=true
   parent=CanvasPanel_Highlights
   widget=Gamepad_Special_Left
   properties {
    bIsEnabled=false
    visibility=Collapsed
    brush {
     bIsDynamicallyLoaded=false
     drawAs=Image
     imageType=NoImage
     mirroring=NoMirror
     resourceName=None
     resourceObject=&M_HUD_Thumbstick_Wide
     tiling=NoTile
     imageSize=<32 32>
     margin=<ltrb 0 0 0 0>
     outlineSettings {
      bUseBrushTransparency=false
      roundingType=HalfHeightRadius
      width=0
      color {colorUseRule=UseColor_Specified, specifiedColor=<rgba 0 0 0 0>}
      cornerRadii=<0 0 0 0>
     }
     tintColor {colorUseRule=UseColor_Specified, specifiedColor=<rgba 1 1 1 1>}
     uVRegion {bIsValid=false, max=<0 0>, min=<0 0>}
    }
   }
   slot {
    class=CanvasPanelSlot
    properties {layoutData {alignment=<0 0>, anchors {maximum=<0.5 0.435>, minimum=<0.5 0.435>}, offsets=<ltrb -44.7489 -47.0015 90.7254 95.3767>}}
   }
  }
  Gamepad_LeftThumbstick {
   widgetClassPath=&Image
   bIsVariable=true
   parent=CanvasPanel_Highlights
   widget=Gamepad_LeftThumbstick
   properties {
    bIsEnabled=false
    visibility=Collapsed
    brush {
     bIsDynamicallyLoaded=false
     drawAs=Image
     imageType=NoImage
     mirroring=NoMirror
     resourceName=None
     resourceObject=&M_HUD_Thumbstick
     tiling=NoTile
     imageSize=<32 32>
     margin=<ltrb 0 0 0 0>
     outlineSettings {
      bUseBrushTransparency=false
      roundingType=HalfHeightRadius
      width=0
      color {colorUseRule=UseColor_Specified, specifiedColor=<rgba 0 0 0 0>}
      cornerRadii=<0 0 0 0>
     }
     tintColor {colorUseRule=UseColor_Specified, specifiedColor=<rgba 1 1 1 1>}
     uVRegion {bIsValid=false, max=<0 0>, min=<0 0>}
    }
   }
   slot {
    class=CanvasPanelSlot
    properties {layoutData {alignment=<0 0>, anchors {maximum=<0.3815 0.7354>, minimum=<0.3815 0.7354>}, offsets=<ltrb -28.4454 -30.2223 60 60>}}
   }
  }
  Gamepad_RightThumbstick {
   widgetClassPath=&Image
   bIsVariable=true
   parent=CanvasPanel_Highlights
   widget=Gamepad_RightThumbstick
   properties {
    bIsEnabled=false
    visibility=Collapsed
    brush {
     bIsDynamicallyLoaded=false
     drawAs=Image
     imageType=NoImage
     mirroring=NoMirror
     resourceName=None
     resourceObject=&M_HUD_Thumbstick
     tiling=NoTile
     imageSize=<32 32>
     margin=<ltrb 0 0 0 0>
     outlineSettings {
      bUseBrushTransparency=false
      roundingType=HalfHeightRadius
      width=0
      color {colorUseRule=UseColor_Specified, specifiedColor=<rgba 0 0 0 0>}
      cornerRadii=<0 0 0 0>
     }
     tintColor {colorUseRule=UseColor_Specified, specifiedColor=<rgba 1 1 1 1>}
     uVRegion {bIsValid=false, max=<0 0>, min=<0 0>}
    }
   }
   slot {
    class=CanvasPanelSlot
    properties {layoutData {alignment=<0 0>, anchors {maximum=<0.6167 0.7366>, minimum=<0.6167 0.7366>}, offsets=<ltrb -28.0006 -35.555 60 60>}}
   }
  }
  Gamepad_DPad_Left {
   widgetClassPath=&Image
   bIsVariable=true
   parent=CanvasPanel_Highlights
   widget=Gamepad_DPad_Left
   properties {
    bIsEnabled=false
    visibility=Collapsed
    brush {
     bIsDynamicallyLoaded=false
     drawAs=Image
     imageType=NoImage
     mirroring=NoMirror
     resourceName=None
     resourceObject=&M_HUD_Thumbstick_Wide
     tiling=NoTile
     imageSize=<32 32>
     margin=<ltrb 0 0 0 0>
     outlineSettings {
      bUseBrushTransparency=false
      roundingType=HalfHeightRadius
      width=0
      color {colorUseRule=UseColor_Specified, specifiedColor=<rgba 0 0 0 0>}
      cornerRadii=<0 0 0 0>
     }
     tintColor {colorUseRule=UseColor_Specified, specifiedColor=<rgba 1 1 1 1>}
     uVRegion {bIsValid=false, max=<0 0>, min=<0 0>}
    }
   }
   slot {
    class=CanvasPanelSlot
    properties {layoutData {alignment=<0 0>, anchors {maximum=<0.2316 0.5233>, minimum=<0.2316 0.5233>}, offsets=<ltrb -32.7411 -25.1856 52 52>}}
   }
  }
  Gamepad_DPad_Right {
   widgetClassPath=&Image
   bIsVariable=true
   parent=CanvasPanel_Highlights
   widget=Gamepad_DPad_Right
   properties {
    bIsEnabled=false
    visibility=Collapsed
    brush {
     bIsDynamicallyLoaded=false
     drawAs=Image
     imageType=NoImage
     mirroring=NoMirror
     resourceName=None
     resourceObject=&M_HUD_Thumbstick_Wide
     tiling=NoTile
     imageSize=<32 32>
     margin=<ltrb 0 0 0 0>
     outlineSettings {
      bUseBrushTransparency=false
      roundingType=HalfHeightRadius
      width=0
      color {colorUseRule=UseColor_Specified, specifiedColor=<rgba 0 0 0 0>}
      cornerRadii=<0 0 0 0>
     }
     tintColor {colorUseRule=UseColor_Specified, specifiedColor=<rgba 1 1 1 1>}
     uVRegion {bIsValid=false, max=<0 0>, min=<0 0>}
    }
   }
   slot {
    class=CanvasPanelSlot
    properties {layoutData {alignment=<0 0>, anchors {maximum=<0.3103 0.5254>, minimum=<0.3103 0.5254>}, offsets=<ltrb -31.8509 -23.4072 52 52>}}
   }
  }
  Gamepad_DPad_Up {
   widgetClassPath=&Image
   bIsVariable=true
   parent=CanvasPanel_Highlights
   widget=Gamepad_DPad_Up
   properties {
    bIsEnabled=false
    visibility=Collapsed
    brush {
     bIsDynamicallyLoaded=false
     drawAs=Image
     imageType=NoImage
     mirroring=NoMirror
     resourceName=None
     resourceObject=&M_HUD_Thumbstick_Wide
     tiling=NoTile
     imageSize=<32 32>
     margin=<ltrb 0 0 0 0>
     outlineSettings {
      bUseBrushTransparency=false
      roundingType=HalfHeightRadius
      width=0
      color {colorUseRule=UseColor_Specified, specifiedColor=<rgba 0 0 0 0>}
      cornerRadii=<0 0 0 0>
     }
     tintColor {colorUseRule=UseColor_Specified, specifiedColor=<rgba 1 1 1 1>}
     uVRegion {bIsValid=false, max=<0 0>, min=<0 0>}
    }
   }
   slot {
    class=CanvasPanelSlot
    properties {layoutData {alignment=<0 0>, anchors {maximum=<0.2694 0.4601>, minimum=<0.2694 0.4601>}, offsets=<ltrb -25.1846 -20.8885 52 52>}}
   }
  }
  Gamepad_DPad_Down {
   widgetClassPath=&Image
   bIsVariable=true
   parent=CanvasPanel_Highlights
   widget=Gamepad_DPad_Down
   properties {
    bIsEnabled=false
    visibility=Collapsed
    brush {
     bIsDynamicallyLoaded=false
     drawAs=Image
     imageType=NoImage
     mirroring=NoMirror
     resourceName=None
     resourceObject=&M_HUD_Thumbstick_Wide
     tiling=NoTile
     imageSize=<32 32>
     margin=<ltrb 0 0 0 0>
     outlineSettings {
      bUseBrushTransparency=false
      roundingType=HalfHeightRadius
      width=0
      color {colorUseRule=UseColor_Specified, specifiedColor=<rgba 0 0 0 0>}
      cornerRadii=<0 0 0 0>
     }
     tintColor {colorUseRule=UseColor_Specified, specifiedColor=<rgba 1 1 1 1>}
     uVRegion {bIsValid=false, max=<0 0>, min=<0 0>}
    }
   }
   slot {
    class=CanvasPanelSlot
    properties {layoutData {alignment=<0 0>, anchors {maximum=<0.2697 0.5793>, minimum=<0.2697 0.5793>}, offsets=<ltrb -25.7779 -17.6289 52 52>}}
   }
  }
  Gamepad_RightShoulder {
   widgetClassPath=&Image
   bIsVariable=true
   parent=CanvasPanel_Highlights
   widget=Gamepad_RightShoulder
   properties {
    bIsEnabled=false
    visibility=Collapsed
    brush {
     bIsDynamicallyLoaded=false
     drawAs=Image
     imageType=NoImage
     mirroring=NoMirror
     resourceName=None
     resourceObject=&M_HUD_Thumbstick
     tiling=NoTile
     imageSize=<32 32>
     margin=<ltrb 0 0 0 0>
     outlineSettings {
      bUseBrushTransparency=false
      roundingType=HalfHeightRadius
      width=0
      color {colorUseRule=UseColor_Specified, specifiedColor=<rgba 0 0 0 0>}
      cornerRadii=<0 0 0 0>
     }
     tintColor {colorUseRule=UseColor_Specified, specifiedColor=<rgba 1 1 1 1>}
     uVRegion {bIsValid=false, max=<0 0>, min=<0 0>}
    }
   }
   slot {
    class=CanvasPanelSlot
    properties {layoutData {alignment=<0 0>, anchors {maximum=<0.7296 0.2199>, minimum=<0.7296 0.2199>}, offsets=<ltrb -28.7399 -21.4812 64 32>}}
   }
  }
  Gamepad_RightTrigger {
   widgetClassPath=&Image
   bIsVariable=true
   parent=CanvasPanel_Highlights
   widget=Gamepad_RightTrigger
   properties {
    bIsEnabled=false
    visibility=Collapsed
    brush {
     bIsDynamicallyLoaded=false
     drawAs=Image
     imageType=NoImage
     mirroring=NoMirror
     resourceName=None
     resourceObject=&M_HUD_Thumbstick
     tiling=NoTile
     imageSize=<32 32>
     margin=<ltrb 0 0 0 0>
     outlineSettings {
      bUseBrushTransparency=false
      roundingType=HalfHeightRadius
      width=0
      color {colorUseRule=UseColor_Specified, specifiedColor=<rgba 0 0 0 0>}
      cornerRadii=<0 0 0 0>
     }
     tintColor {colorUseRule=UseColor_Specified, specifiedColor=<rgba 1 1 1 1>}
     uVRegion {bIsValid=false, max=<0 0>, min=<0 0>}
    }
   }
   slot {
    class=CanvasPanelSlot
    properties {layoutData {alignment=<0 0>, anchors {maximum=<0.7219 0.1291>, minimum=<0.7219 0.1291>}, offsets=<ltrb -19.4613 -23.5393 32 32>}}
   }
  }
  Gamepad_LeftShoulder {
   widgetClassPath=&Image
   bIsVariable=true
   parent=CanvasPanel_Highlights
   widget=Gamepad_LeftShoulder
   properties {
    bIsEnabled=false
    visibility=Collapsed
    brush {
     bIsDynamicallyLoaded=false
     drawAs=Image
     imageType=NoImage
     mirroring=NoMirror
     resourceName=None
     resourceObject=&M_HUD_Thumbstick
     tiling=NoTile
     imageSize=<32 32>
     margin=<ltrb 0 0 0 0>
     outlineSettings {
      bUseBrushTransparency=false
      roundingType=HalfHeightRadius
      width=0
      color {colorUseRule=UseColor_Specified, specifiedColor=<rgba 0 0 0 0>}
      cornerRadii=<0 0 0 0>
     }
     tintColor {colorUseRule=UseColor_Specified, specifiedColor=<rgba 1 1 1 1>}
     uVRegion {bIsValid=false, max=<0 0>, min=<0 0>}
    }
   }
   slot {
    class=CanvasPanelSlot
    properties {layoutData {alignment=<0 0>, anchors {maximum=<0.268 0.2247>, minimum=<0.268 0.2247>}, offsets=<ltrb -34.5178 -22.6663 64 32>}}
   }
  }
  Gamepad_LeftTrigger {
   widgetClassPath=&Image
   bIsVariable=true
   parent=CanvasPanel_Highlights
   widget=Gamepad_LeftTrigger
   properties {
    bIsEnabled=false
    visibility=Collapsed
    brush {
     bIsDynamicallyLoaded=false
     drawAs=Image
     imageType=NoImage
     mirroring=NoMirror
     resourceName=None
     resourceObject=&M_HUD_Thumbstick
     tiling=NoTile
     imageSize=<32 32>
     margin=<ltrb 0 0 0 0>
     outlineSettings {
      bUseBrushTransparency=false
      roundingType=HalfHeightRadius
      width=0
      color {colorUseRule=UseColor_Specified, specifiedColor=<rgba 0 0 0 0>}
      cornerRadii=<0 0 0 0>
     }
     tintColor {colorUseRule=UseColor_Specified, specifiedColor=<rgba 1 1 1 1>}
     uVRegion {bIsValid=false, max=<0 0>, min=<0 0>}
    }
   }
   slot {
    class=CanvasPanelSlot
    properties {layoutData {alignment=<0 0>, anchors {maximum=<0.2792 0.1198>, minimum=<0.2792 0.1198>}, offsets=<ltrb -20.0006 -9.3332 32 32>}}
   }
  }
 ]
}
