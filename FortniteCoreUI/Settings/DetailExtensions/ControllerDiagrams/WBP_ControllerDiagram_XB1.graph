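# Xbox One controller diagram widget for the settings detail extensions, built on FortControllerHighlightDiagram to highlight gamepad inputs.
# An XB1 controller image sits under a visibility switcher holding the inherited highlight canvas.
# That canvas anchors collapsed HUD thumbstick-material images, named per gamepad input, over each button, stick, trigger, shoulder and D-pad direction.
# The event graph holds only empty event stubs, so the Blueprint supplies layout and no highlight logic.
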WidgetBlueprint /FortniteCoreUI/Settings/DetailExtensions/ControllerDiagrams/WBP_ControllerDiagram_XB1
graphs [EventGraph {nodes [
   Event PreConstruct()
   Event Construct()
   Event Tick()
  ]}]
variables []
widgets {
 info {inheritedWidgetCount=1, parentClass=/Script/FortniteUI.FortControllerHighlightDiagram, rootWidgetClass=&SizeBox, widgetCount=21}
 widgets [
  SizeBox_0 {widgetClassPath=&SizeBox, widget=SizeBox_0}
  Overlay_30 {widgetClassPath=&Overlay, parent=SizeBox_0, widget=Overlay_30, slot {class=SizeBoxSlot}}
  Image_ControllerInput {
   widgetClassPath=&Image
   bIsVariable=true
   parent=Overlay_30
   widget=Image_ControllerInput
   properties {
    bIsEnabled=false
    visibility=SelfHitTestInvisible
    brush {
     bIsDynamicallyLoaded=false
     drawAs=Image
     imageType=NoImage
     mirroring=NoMirror
     resourceName=None
     resourceObject=/FortUILibrary/Input/GamepadXboxOne/XB1_ControllerConfig
     tiling=NoTile
     imageSize=<1024 512>
     margin=<ltrb 0 0 0 0>
     outlineSettings {
      bUseBrushTransparency=false
      roundingType=HalfHeightRadius
      width=0
      color {colorUseRule=UseColor_Specified, specifiedColor=<rgba 0 0 0 0>}
      cornerRadii=<0 0 0 0>
     }
     tintColor {colorUseRule=UseColor_Specified, specifiedColor=<rgba 1 1 1 1>}
     uVRegion {bIsValid=false, max=<0 0>, min=<0 0>}
    }
   }
   slot {class=OverlaySlot, properties {horizontalAlignment=HAlign_Fill, verticalAlignment=VAlign_Fill}}
  }
  Switcher_HighlightedElement {
   widgetClassPath=&CommonVisibilitySwitcher
   bIsVariable=true
   parent=Overlay_30
   widget=Switcher_HighlightedElement
   slot {class=OverlaySlot, properties {horizontalAlignment=HAlign_Fill, verticalAlignment=VAlign_Fill}}
  }
  CanvasPanel_Highlights {
   widgetClassPath=&CanvasPanel
   bInherited=true
   parent=Switcher_HighlightedElement
   widget=CanvasPanel_Highlights
   slot {class=CommonVisibilitySwitcherSlot}
  }
  Gamepad_FaceButton_Right {
   widgetClassPath=&Image
   bIsVariable=true
   parent=CanvasPanel_Highlights
   widget=Gamepad_FaceButton_Right
   properties {
    bIsEnabled=false
    visibility=Collapsed
    brush {
     bIsDynamicallyLoaded=false
     drawAs=Image
     imageType=NoImage
     mirroring=NoMirror
     resourceName=None
     resourceObject=/Game/Athena/HUD/Art/Materials/M_HUD_Thumbstick_Wide
     tiling=NoTile
     imageSize=<32 32>
     margin=<ltrb 0 0 0 0>
     outlineSettings {
      bUseBrushTransparency=false
      roundingType=HalfHeightRadius
      width=0
      color {colorUseRule=UseColor_Specified, specifiedColor=<rgba 0 0 0 0>}
      cornerRadii=<0 0 0 0>
     }
     tintColor {colorUseRule=UseColor_Specified, specifiedColor=<rgba 1 1 1 1>}
     uVRegion {bIsValid=false, max=<0 0>, min=<0 0>}
    }
   }
   slot {
    class=CanvasPanelSlot
    properties {layoutData {alignment=<0 0>, anchors {maximum=<0.7151 0.3245>, minimum=<0.7151 0.3245>}, offsets=<ltrb -24.9382 -26.4449 52 52>}}
   }
  }
  Gamepad_FaceButton_Left {
   widgetClassPath=&Image
   bIsVariable=true
   parent=CanvasPanel_Highlights
   widget=Gamepad_FaceButton_Left
   properties {
    bIsEnabled=false
    visibility=Collapsed
    brush {
     bIsDynamicallyLoaded=false
     drawAs=Image
     imageType=NoImage
     mirroring=NoMirror
     resourceName=None
     resourceObject=&M_HUD_Thumbstick_Wide
     tiling=NoTile
     imageSize=<64 64>
     margin=<ltrb 0 0 0 0>
     outlineSettings {
      bUseBrushTransparency=false
      roundingType=HalfHeightRadius
      width=0
      color {colorUseRule=UseColor_Specified, specifiedColor=<rgba 0 0 0 0>}
      cornerRadii=<0 0 0 0>
     }
     tintColor {colorUseRule=UseColor_Specified, specifiedColor=<rgba 1 1 1 1>}
     uVRegion {bIsValid=false, max=<0 0>, min=<0 0>}
    }
   }
   slot {
    class=CanvasPanelSlot
    properties {layoutData {alignment=<0 0>, anchors {maximum=<0.6235 0.328>, minimum=<0.6235 0.328>}, offsets=<ltrb -25.0547 -30.2216 52 52>}}
   }
  }
  Gamepad_FaceButton_Top {
   widgetClassPath=&Image
   bIsVariable=true
   parent=CanvasPanel_Highlights
   widget=Gamepad_FaceButton_Top
   properties {
    bIsEnabled=false
    visibility=Collapsed
    brush {
     bIsDynamicallyLoaded=false
     drawAs=Image
     imageType=NoImage
     mirroring=NoMirror
     resourceName=None
     resourceObject=&M_HUD_Thumbstick_Wide
     tiling=NoTile
     imageSize=<64 64>
     margin=<ltrb 0 0 0 0>
     outlineSettings {
      bUseBrushTransparency=false
      roundingType=HalfHeightRadius
      width=0
      color {colorUseRule=UseColor_Specified, specifiedColor=<rgba 0 0 0 0>}
      cornerRadii=<0 0 0 0>
     }
     tintColor {colorUseRule=UseColor_Specified, specifiedColor=<rgba 1 1 1 1>}
     uVRegion {bIsValid=false, max=<0 0>, min=<0 0>}
    }
   }
   slot {
    class=CanvasPanelSlot
    properties {layoutData {alignment=<0 0>, anchors {maximum=<0.6691 0.2363>, minimum=<0.6691 0.2363>}, offsets=<ltrb -24.7411 -27.1522 52 52>}}
   }
  }
  Gamepad_FaceButton_Bottom {
   widgetClassPath=&Image
   bIsVariable=true
   parent=CanvasPanel_Highlights
   widget=Gamepad_FaceButton_Bottom
   properties {
    bIsEnabled=false
    visibility=Collapsed
    brush {
     bIsDynamicallyLoaded=false
     drawAs=Image
     imageType=NoImage
     mirroring=NoMirror
     resourceName=None
     resourceObject=&M_HUD_Thumbstick_Wide
     tiling=NoTile
     imageSize=<64 64>
     margin=<ltrb 0 0 0 0>
     outlineSettings {
      bUseBrushTransparency=false
      roundingType=HalfHeightRadius
      width=0
      color {colorUseRule=UseColor_Specified, specifiedColor=<rgba 0 0 0 0>}
      cornerRadii=<0 0 0 0>
     }
     tintColor {colorUseRule=UseColor_Specified, specifiedColor=<rgba 1 1 1 1>}
     uVRegion {bIsValid=false, max=<0 0>, min=<0 0>}
    }
   }
   slot {
    class=CanvasPanelSlot
    properties {layoutData {alignment=<0 0>, anchors {maximum=<0.6688 0.4132>, minimum=<0.6688 0.4132>}, offsets=<ltrb -24.0134 -22.2214 52 52>}}
   }
  }
  Gamepad_Special_Right {
   widgetClassPath=&Image
   bIsVariable=true
   parent=CanvasPanel_Highlights
   widget=Gamepad_Special_Right
   properties {
    bIsEnabled=false
    visibility=Collapsed
    brush {
     bIsDynamicallyLoaded=false
     drawAs=Image
     imageType=NoImage
     mirroring=NoMirror
     resourceName=None
     resourceObject=/Game/Athena/HUD/Art/Materials/M_HUD_Thumbstick
     tiling=NoTile
     imageSize=<32 32>
     margin=<ltrb 0 0 0 0>
     outlineSettings {
      bUseBrushTransparency=false
      roundingType=HalfHeightRadius
      width=0
      color {colorUseRule=UseColor_Specified, specifiedColor=<rgba 0 0 0 0>}
      cornerRadii=<0 0 0 0>
     }
     tintColor {colorUseRule=UseColor_Specified, specifiedColor=<rgba 1 1 1 1>}
     uVRegion {bIsValid=false, max=<0 0>, min=<0 0>}
    }
   }
   slot {
    class=CanvasPanelSlot
    properties {layoutData {alignment=<0 0>, anchors {maximum=<0.5473 0.3233>, minimum=<0.5473 0.3233>}, offsets=<ltrb -14.7393 -13.1856 32 32>}}
   }
  }
  Gamepad_Special_Left {
   widgetClassPath=&Image
   bIsVariable=true
   parent=CanvasPanel_Highlights
   widget=Gamepad_Special_Left
   properties {
    bIsEnabled=false
    visibility=Collapsed
    brush {
     bIsDynamicallyLoaded=false
     drawAs=Image
     imageType=NoImage
     mirroring=NoMirror
     resourceName=None
     resourceObject=&M_HUD_Thumbstick
     tiling=NoTile
     imageSize=<32 32>
     margin=<ltrb 0 0 0 0>
     outlineSettings {
      bUseBrushTransparency=false
      roundingType=HalfHeightRadius
      width=0
      color {colorUseRule=UseColor_Specified, specifiedColor=<rgba 0 0 0 0>}
      cornerRadii=<0 0 0 0>
     }
     tintColor {colorUseRule=UseColor_Specified, specifiedColor=<rgba 1 1 1 1>}
     uVRegion {bIsValid=false, max=<0 0>, min=<0 0>}
    }
   }
   slot {
    class=CanvasPanelSlot
    properties {layoutData {alignment=<0 0>, anchors {maximum=<0.4545 0.3243>, minimum=<0.4545 0.3243>}, offsets=<ltrb -16.5551 -14.2588 32 32>}}
   }
  }
  Gamepad_LeftThumbstick {
   widgetClassPath=&Image
   bIsVariable=true
   parent=CanvasPanel_Highlights
   widget=Gamepad_LeftThumbstick
   properties {
    bIsEnabled=false
    visibility=Collapsed
    brush {
     bIsDynamicallyLoaded=false
     drawAs=Image
     imageType=NoImage
     mirroring=NoMirror
     resourceName=None
     resourceObject=&M_HUD_Thumbstick
     tiling=NoTile
     imageSize=<32 32>
     margin=<ltrb 0 0 0 0>
     outlineSettings {
      bUseBrushTransparency=false
      roundingType=HalfHeightRadius
      width=0
      color {colorUseRule=UseColor_Specified, specifiedColor=<rgba 0 0 0 0>}
      cornerRadii=<0 0 0 0>
     }
     tintColor {colorUseRule=UseColor_Specified, specifiedColor=<rgba 1 1 1 1>}
     uVRegion {bIsValid=false, max=<0 0>, min=<0 0>}
    }
   }
   slot {
    class=CanvasPanelSlot
    properties {layoutData {alignment=<0 0>, anchors {maximum=<0.3325 0.323>, minimum=<0.3325 0.323>}, offsets=<ltrb -32 -28.8886 60 60>}}
   }
  }
  Gamepad_RightThumbstick {
   widgetClassPath=&Image
   bIsVariable=true
   parent=CanvasPanel_Highlights
   widget=Gamepad_RightThumbstick
   properties {
    bIsEnabled=false
    visibility=Collapsed
    brush {
     bIsDynamicallyLoaded=false
     drawAs=Image
     imageType=NoImage
     mirroring=NoMirror
     resourceName=None
     resourceObject=&M_HUD_Thumbstick
     tiling=NoTile
     imageSize=<32 32>
     margin=<ltrb 0 0 0 0>
     outlineSettings {
      bUseBrushTransparency=false
      roundingType=HalfHeightRadius
      width=0
      color {colorUseRule=UseColor_Specified, specifiedColor=<rgba 0 0 0 0>}
      cornerRadii=<0 0 0 0>
     }
     tintColor {colorUseRule=UseColor_Specified, specifiedColor=<rgba 1 1 1 1>}
     uVRegion {bIsValid=false, max=<0 0>, min=<0 0>}
    }
   }
   slot {
    class=CanvasPanelSlot
    properties {layoutData {alignment=<0 0>, anchors {maximum=<0.5847 0.5237>, minimum=<0.5847 0.5237>}, offsets=<ltrb -30.6682 -29.5549 60 60>}}
   }
  }
  Gamepad_DPad_Left {
   widgetClassPath=&Image
   bIsVariable=true
   parent=CanvasPanel_Highlights
   widget=Gamepad_DPad_Left
   properties {
    bIsEnabled=false
    visibility=Collapsed
    brush {
     bIsDynamicallyLoaded=false
     drawAs=Image
     imageType=NoImage
     mirroring=NoMirror
     resourceName=None
     resourceObject=&M_HUD_Thumbstick_Wide
     tiling=NoTile
     imageSize=<32 32>
     margin=<ltrb 0 0 0 0>
     outlineSettings {
      bUseBrushTransparency=false
      roundingType=HalfHeightRadius
      width=0
      color {colorUseRule=UseColor_Specified, specifiedColor=<rgba 0 0 0 0>}
      cornerRadii=<0 0 0 0>
     }
     tintColor {colorUseRule=UseColor_Specified, specifiedColor=<rgba 1 1 1 1>}
     uVRegion {bIsValid=false, max=<0 0>, min=<0 0>}
    }
   }
   slot {
    class=CanvasPanelSlot
    properties {layoutData {alignment=<0 0>, anchors {maximum=<0.3844 0.5357>, minimum=<0.3844 0.5357>}, offsets=<ltrb -22.0749 -26.5193 52 52>}}
   }
  }
  Gamepad_DPad_Right {
   widgetClassPath=&Image
   bIsVariable=true
   parent=CanvasPanel_Highlights
   widget=Gamepad_DPad_Right
   properties {
    bIsEnabled=false
    visibility=Collapsed
    brush {
     bIsDynamicallyLoaded=false
     drawAs=Image
     imageType=NoImage
     mirroring=NoMirror
     resourceName=None
     resourceObject=&M_HUD_Thumbstick_Wide
     tiling=NoTile
     imageSize=<32 32>
     margin=<ltrb 0 0 0 0>
     outlineSettings {
      bUseBrushTransparency=false
      roundingType=HalfHeightRadius
      width=0
      color {colorUseRule=UseColor_Specified, specifiedColor=<rgba 0 0 0 0>}
      cornerRadii=<0 0 0 0>
     }
     tintColor {colorUseRule=UseColor_Specified, specifiedColor=<rgba 1 1 1 1>}
     uVRegion {bIsValid=false, max=<0 0>, min=<0 0>}
    }
   }
   slot {
    class=CanvasPanelSlot
    properties {layoutData {alignment=<0 0>, anchors {maximum=<0.4557 0.5337>, minimum=<0.4557 0.5337>}, offsets=<ltrb -38.8519 -24.4068 52 52>}}
   }
  }
  Gamepad_DPad_Up {
   widgetClassPath=&Image
   bIsVariable=true
   parent=CanvasPanel_Highlights
   widget=Gamepad_DPad_Up
   properties {
    bIsEnabled=false
    visibility=Collapsed
    brush {
     bIsDynamicallyLoaded=false
     drawAs=Image
     imageType=NoImage
     mirroring=NoMirror
     resourceName=None
     resourceObject=&M_HUD_Thumbstick_Wide
     tiling=NoTile
     imageSize=<32 32>
     margin=<ltrb 0 0 0 0>
     outlineSettings {
      bUseBrushTransparency=false
      roundingType=HalfHeightRadius
      width=0
      color {colorUseRule=UseColor_Specified, specifiedColor=<rgba 0 0 0 0>}
      cornerRadii=<0 0 0 0>
     }
     tintColor {colorUseRule=UseColor_Specified, specifiedColor=<rgba 1 1 1 1>}
     uVRegion {bIsValid=false, max=<0 0>, min=<0 0>}
    }
   }
   slot {
    class=CanvasPanelSlot
    properties {layoutData {alignment=<0 0>, anchors {maximum=<0.4155 0.4626>, minimum=<0.4155 0.4626>}, offsets=<ltrb -25.8522 -19.5551 52 52>}}
   }
  }
  Gamepad_DPad_Down {
   widgetClassPath=&Image
   bIsVariable=true
   parent=CanvasPanel_Highlights
   widget=Gamepad_DPad_Down
   properties {
    bIsEnabled=false
    visibility=Collapsed
    brush {
     bIsDynamicallyLoaded=false
     drawAs=Image
     imageType=NoImage
     mirroring=NoMirror
     resourceName=None
     resourceObject=&M_HUD_Thumbstick_Wide
     tiling=NoTile
     imageSize=<32 32>
     margin=<ltrb 0 0 0 0>
     outlineSettings {
      bUseBrushTransparency=false
      roundingType=HalfHeightRadius
      width=0
      color {colorUseRule=UseColor_Specified, specifiedColor=<rgba 0 0 0 0>}
      cornerRadii=<0 0 0 0>
     }
     tintColor {colorUseRule=UseColor_Specified, specifiedColor=<rgba 1 1 1 1>}
     uVRegion {bIsValid=false, max=<0 0>, min=<0 0>}
    }
   }
   slot {
    class=CanvasPanelSlot
    properties {layoutData {alignment=<0 0>, anchors {maximum=<0.4159 0.6>, minimum=<0.4159 0.6>}, offsets=<ltrb -26.4454 -28 52 52>}}
   }
  }
  Gamepad_RightShoulder {
   widgetClassPath=&Image
   bIsVariable=true
   parent=CanvasPanel_Highlights
   widget=Gamepad_RightShoulder
   properties {
    bIsEnabled=false
    visibility=Collapsed
    brush {
     bIsDynamicallyLoaded=false
     drawAs=Image
     imageType=NoImage
     mirroring=NoMirror
     resourceName=None
     resourceObject=&M_HUD_Thumbstick
     tiling=NoTile
     imageSize=<32 32>
     margin=<ltrb 0 0 0 0>
     outlineSettings {
      bUseBrushTransparency=false
      roundingType=HalfHeightRadius
      width=0
      color {colorUseRule=UseColor_Specified, specifiedColor=<rgba 0 0 0 0>}
      cornerRadii=<0 0 0 0>
     }
     tintColor {colorUseRule=UseColor_Specified, specifiedColor=<rgba 1 1 1 1>}
     uVRegion {bIsValid=false, max=<0 0>, min=<0 0>}
    }
   }
   slot {
    class=CanvasPanelSlot
    properties {layoutData {alignment=<0 0>, anchors {maximum=<0.6594 0.1>, minimum=<0.6594 0.1>}, offsets=<ltrb -34.073 -24 64 32>}}
   }
  }
  Gamepad_RightTrigger {
   widgetClassPath=&Image
   bIsVariable=true
   parent=CanvasPanel_Highlights
   widget=Gamepad_RightTrigger
   properties {
    bIsEnabled=false
    visibility=Collapsed
    brush {
     bIsDynamicallyLoaded=false
     drawAs=Image
     imageType=NoImage
     mirroring=NoMirror
     resourceName=None
     resourceObject=&M_HUD_Thumbstick
     tiling=NoTile
     imageSize=<32 32>
     margin=<ltrb 0 0 0 0>
     outlineSettings {
      bUseBrushTransparency=false
      roundingType=HalfHeightRadius
      width=0
      color {colorUseRule=UseColor_Specified, specifiedColor=<rgba 0 0 0 0>}
      cornerRadii=<0 0 0 0>
     }
     tintColor {colorUseRule=UseColor_Specified, specifiedColor=<rgba 1 1 1 1>}
     uVRegion {bIsValid=false, max=<0 0>, min=<0 0>}
    }
   }
   slot {
    class=CanvasPanelSlot
    properties {layoutData {alignment=<0 0>, anchors {maximum=<0.6622 0.0396>, minimum=<0.6622 0.0396>}, offsets=<ltrb -15.4087 -14.741 32 32>}}
   }
  }
  Gamepad_LeftShoulder {
   widgetClassPath=&Image
   bIsVariable=true
   parent=CanvasPanel_Highlights
   widget=Gamepad_LeftShoulder
   properties {
    bIsEnabled=false
    visibility=Collapsed
    brush {
     bIsDynamicallyLoaded=false
     drawAs=Image
     imageType=NoImage
     mirroring=NoMirror
     resourceName=None
     resourceObject=&M_HUD_Thumbstick
     tiling=NoTile
     imageSize=<32 32>
     margin=<ltrb 0 0 0 0>
     outlineSettings {
      bUseBrushTransparency=false
      roundingType=HalfHeightRadius
      width=0
      color {colorUseRule=UseColor_Specified, specifiedColor=<rgba 0 0 0 0>}
      cornerRadii=<0 0 0 0>
     }
     tintColor {colorUseRule=UseColor_Specified, specifiedColor=<rgba 1 1 1 1>}
     uVRegion {bIsValid=false, max=<0 0>, min=<0 0>}
    }
   }
   slot {
    class=CanvasPanelSlot
    properties {layoutData {alignment=<0 0>, anchors {maximum=<0.3374 0.1>, minimum=<0.3374 0.1>}, offsets=<ltrb -35.8502 -24 64 32>}}
   }
  }
  Gamepad_LeftTrigger {
   widgetClassPath=&Image
   bIsVariable=true
   parent=CanvasPanel_Highlights
   widget=Gamepad_LeftTrigger
   properties {
    bIsEnabled=false
    visibility=Collapsed
    brush {
     bIsDynamicallyLoaded=false
     drawAs=Image
     imageType=NoImage
     mirroring=NoMirror
     resourceName=None
     resourceObject=&M_HUD_Thumbstick
     tiling=NoTile
     imageSize=<32 32>
     margin=<ltrb 0 0 0 0>
     outlineSettings {
      bUseBrushTransparency=false
      roundingType=HalfHeightRadius
      width=0
      color {colorUseRule=UseColor_Specified, specifiedColor=<rgba 0 0 0 0>}
      cornerRadii=<0 0 0 0>
     }
     tintColor {colorUseRule=UseColor_Specified, specifiedColor=<rgba 1 1 1 1>}
     uVRegion {bIsValid=false, max=<0 0>, min=<0 0>}
    }
   }
   slot {
    class=CanvasPanelSlot
    properties {layoutData {alignment=<0 0>, anchors {maximum=<0.3299 0.0444>, minimum=<0.3299 0.0444>}, offsets=<ltrb -13.3331 -15.9995 32 32>}}
   }
  }
 ]
}
