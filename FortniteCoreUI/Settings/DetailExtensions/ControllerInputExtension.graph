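# Settings detail extension widget that shows a controller diagram for controller input settings.
# A fixed-size box and scale box wrap a visibility switcher holding diagram widgets for PS4, PS5, XB1, XSX, Erebus, Ounce and a generic pad.
# The switcher and diagrams bind to the native FortSettingDetailExtension_ControllerInput parent.
# The Blueprint supplies only layout, with an empty event graph and no diagram selection logic.
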WidgetBlueprint /FortniteCoreUI/Settings/DetailExtensions/ControllerInputExtension
graphs [EventGraph {links [], nodes []}]
variables []
widgets {
 info {
  inheritedWidgetCount=8
  parentClass=/Script/FortniteUI.FortSettingDetailExtension_ControllerInput
  rootWidgetClass=&SizeBox
  widgetCount=10
 }
 widgets [
  SizeBox_0 {
   widgetClassPath=&SizeBox
   widget=SizeBox_0
   properties {bOverride_HeightOverride=true, bOverride_WidthOverride=true, heightOverride=500, widthOverride=700}
  }
  ScaleBox_0 {widgetClassPath=&ScaleBox, parent=SizeBox_0, widget=ScaleBox_0, slot {class=SizeBoxSlot}}
  Switcher_ControllerDiagram {
   widgetClassPath=&CommonVisibilitySwitcher
   bInherited=true
   parent=ScaleBox_0
   widget=Switcher_ControllerDiagram
   slot {class=ScaleBoxSlot}
  }
  ControllerDiagram_PS4 {
   widgetClassPath=/FortniteCoreUI/Settings/DetailExtensions/ControllerDiagrams/WBP_ControllerDiagram_PS4
   bInherited=true
   bIsVariable=true
   parent=Switcher_ControllerDiagram
   widget=ControllerDiagram_PS4
   slot {class=CommonVisibilitySwitcherSlot}
  }
  ControllerDiagram_PS5 {
   widgetClassPath=/FortniteCoreUI/Settings/DetailExtensions/ControllerDiagrams/WBP_ControllerDiagram_PS5
   bInherited=true
   bIsVariable=true
   parent=Switcher_ControllerDiagram
   widget=ControllerDiagram_PS5
   slot {class=CommonVisibilitySwitcherSlot}
  }
  ControllerDiagram_XB1 {
   widgetClassPath=/FortniteCoreUI/Settings/DetailExtensions/ControllerDiagrams/WBP_ControllerDiagram_XB1
   bInherited=true
   bIsVariable=true
   parent=Switcher_ControllerDiagram
   widget=ControllerDiagram_XB1
   slot {class=CommonVisibilitySwitcherSlot}
  }
  ControllerDiagram_XSX {
   widgetClassPath=/FortniteCoreUI/Settings/DetailExtensions/ControllerDiagrams/WBP_ControllerDiagram_XSX
   bInherited=true
   bIsVariable=true
   parent=Switcher_ControllerDiagram
   widget=ControllerDiagram_XSX
   slot {class=CommonVisibilitySwitcherSlot}
  }
  ControllerDiagram_Erebus {
   widgetClassPath=/FortniteCoreUI/Settings/DetailExtensions/ControllerDiagrams/WBP_ControllerDiagram_Erebus
   bInherited=true
   bIsVariable=true
   parent=Switcher_ControllerDiagram
   widget=ControllerDiagram_Erebus
   slot {class=CommonVisibilitySwitcherSlot}
  }
  ControllerDiagram_Ounce {
   widgetClassPath=/FortniteCoreUI/Settings/DetailExtensions/ControllerDiagrams/WBP_ControllerDiagram_Ounce
   bInherited=true
   bIsVariable=true
   parent=Switcher_ControllerDiagram
   widget=ControllerDiagram_Ounce
   slot {class=CommonVisibilitySwitcherSlot}
  }
  ControllerDiagram_Generic {
   widgetClassPath=/FortniteCoreUI/Settings/DetailExtensions/ControllerDiagrams/WBP_ControllerDiagram_Generic
   bInherited=true
   bIsVariable=true
   parent=Switcher_ControllerDiagram
   widget=ControllerDiagram_Generic
   slot {class=CommonVisibilitySwitcherSlot}
  }
 ]
}
