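# Settings detail widget that visualizes a controller stick deadzone, based on FortSettingsDeadzoneExtension.
# An overlay stacks three centered thumbstick-background images: a white maximum, a blue deadzone adjustment and a yellow live value, the latter two starting at zero scale.
# Every Tick, the parent's live and deadzone analog values become uniform render scales for the yellow and blue images, so their sizes compare against the fixed white one.
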WidgetBlueprint /FortniteCoreUI/Settings/DetailExtensions/DeadzoneExtension
graphs [EventGraph {nodes [
   Event Tick()
   n1 = GetLiveAnalogValue!(self:Fort Settings Deadzone Extension Object Reference) -> (ReturnValue:float)
   > SetRenderScale!(
    execute
    self:Widget Object Reference=Get Deadzone_LiveValue() -> (Deadzone_LiveValue:Image Object Reference)
    Scale:Vector 2D Structure=MakeVector2D(X=@n1, Y=@n1)
   ) -> (then)
   n2 = GetDeadZoneAnalogValue!(self:Fort Settings Deadzone Extension Object Reference) -> (ReturnValue:float)
   > SetRenderScale(self=Get Deadzone_Adjustment() -> (Deadzone_Adjustment:Image Object Reference), Scale=MakeVector2D(X=@n2, Y=@n2))
  ]}]
variables []
widgets {
 info {inheritedWidgetCount=0, parentClass=/Script/FortniteUI.FortSettingsDeadzoneExtension, rootWidgetClass=&Overlay, widgetCount=4}
 widgets [
  Overlay_2 {widgetClassPath=&Overlay, widget=Overlay_2}
  Deadzone_Max {
   widgetClassPath=&Image
   bIsVariable=true
   parent=Overlay_2
   widget=Deadzone_Max
   properties {
    visibility=SelfHitTestInvisible
    brush {
     bIsDynamicallyLoaded=false
     drawAs=Image
     imageType=NoImage
     mirroring=NoMirror
     resourceName=None
     resourceObject=/Game/Athena/HUD/Art/Materials/MI_Hud_Thumbstick_BG_Inst
     tiling=NoTile
     imageSize=<300 300>
     margin=<ltrb 0 0 0 0>
     outlineSettings {
      bUseBrushTransparency=false
      roundingType=HalfHeightRadius
      width=0
      color {colorUseRule=UseColor_Specified, specifiedColor=<rgba 0 0 0 0>}
      cornerRadii=<0 0 0 0>
     }
     tintColor {colorUseRule=UseColor_Specified, specifiedColor=<rgba 1 1 1 1>}
     uVRegion {bIsValid=false, max=<0 0>, min=<0 0>}
    }
   }
   slot {class=OverlaySlot, properties {horizontalAlignment=HAlign_Center, verticalAlignment=VAlign_Center}}
  }
  Deadzone_Adjustment {
   widgetClassPath=&Image
   bIsVariable=true
   parent=Overlay_2
   widget=Deadzone_Adjustment
   properties {
    visibility=SelfHitTestInvisible
    brush {
     bIsDynamicallyLoaded=false
     drawAs=Image
     imageType=NoImage
     mirroring=NoMirror
     resourceName=None
     resourceObject=&MI_Hud_Thumbstick_BG_Inst
     tiling=NoTile
     imageSize=<300 300>
     margin=<ltrb 0 0 0 0>
     outlineSettings {
      bUseBrushTransparency=false
      roundingType=HalfHeightRadius
      width=0
      color {colorUseRule=UseColor_Specified, specifiedColor=<rgba 0 0 0 0>}
      cornerRadii=<0 0 0 0>
     }
     tintColor {colorUseRule=UseColor_Specified, specifiedColor=<rgba 0 0.1164 1 1>}
     uVRegion {bIsValid=false, max=<0 0>, min=<0 0>}
    }
    renderTransform {angle=0, scale=<0 0>, shear=<0 0>, translation=<0 0>}
   }
   slot {class=OverlaySlot, properties {horizontalAlignment=HAlign_Center, verticalAlignment=VAlign_Center}}
  }
  Deadzone_LiveValue {
   widgetClassPath=&Image
   bIsVariable=true
   parent=Overlay_2
   widget=Deadzone_LiveValue
   properties {
    visibility=SelfHitTestInvisible
    brush {
     bIsDynamicallyLoaded=false
     drawAs=Image
     imageType=NoImage
     mirroring=NoMirror
     resourceName=None
     resourceObject=&MI_Hud_Thumbstick_BG_Inst
     tiling=NoTile
     imageSize=<300 300>
     margin=<ltrb 0 0 0 0>
     outlineSettings {
      bUseBrushTransparency=false
      roundingType=HalfHeightRadius
      width=0
      color {colorUseRule=UseColor_Specified, specifiedColor=<rgba 0 0 0 0>}
      cornerRadii=<0 0 0 0>
     }
     tintColor {colorUseRule=UseColor_Specified, specifiedColor=<rgba 1 0.7929 0 1>}
     uVRegion {bIsValid=false, max=<0 0>, min=<0 0>}
    }
    renderTransform {angle=0, scale=<0 0>, shear=<0 0>, translation=<0 0>}
   }
   slot {class=OverlaySlot, properties {horizontalAlignment=HAlign_Center, verticalAlignment=VAlign_Center}}
  }
 ]
}
