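# DelayedHDTexturesStatusExtension is a settings detail extension widget, named for showing the status of delayed HD texture content.
# It derives from the native FortSettingDetailExtension class, and its widget tree holds only an empty root Overlay.
# The event graph has no nodes and the Blueprint declares no variables, so it adds no visuals or logic beyond the parent class.
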WidgetBlueprint /FortniteCoreUI/Settings/DetailExtensions/DelayedHDTexturesStatusExtension
graphs [EventGraph {links [], nodes []}]
variables []
widgets {
 info {inheritedWidgetCount=0, parentClass=/Script/FortniteUI.FortSettingDetailExtension, rootWidgetClass=&Overlay, widgetCount=1}
 widgets [Overlay_0 {widgetClassPath=&Overlay, widget=Overlay_0}]
}
