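# Widget for a single enum option row in a settings detail extension, showing an uppercase label with an arrow that marks the selected option.
# SetupEntry receives the text and a Selected flag, sets the label, and switches its color between DefaultColor and SelectedColor.
# The same flag shows or hides and tints the arrow, and toggles the border background's alpha while keeping its RGB.
# The widget only handles presentation, with no input or selection logic.
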WidgetBlueprint /FortniteCoreUI/Settings/DetailExtensions/EnumOptionDetailsEntry
graphs [EventGraph {nodes [
   n1 = SetupEntry!() -> (OutputDelegate, then, Text:Text, Selected:bool=~) {is=title K2Node_CustomEvent}
   > "SetText (Text)"(self=Get CommonTextBlock_34() -> (CommonTextBlock_34:Common Text Object Reference), InText=@n1.Text)
   > SetVisibility(
    self=Get Image_60() -> (Image_60:Image Object Reference)
    InVisibility=Select(Option 0=Hidden, Option 1=SelfHitTestInvisible, Index=@n1.Selected)
   )
   > SetColorAndOpacity(
    self=Get CommonTextBlock_34()
    InColorAndOpacity=Select(
     Option 0:Slate Color Structure=Get DefaultColor() -> (DefaultColor:Slate Color Structure)
     Option 1:Slate Color Structure=Get SelectedColor() -> (SelectedColor:Slate Color Structure)
     Index=@n1.Selected
    ) -> (ReturnValue:Slate Color Structure)
   )
   > SetColorAndOpacity(
    self:Image Object Reference=Get Image_60()
    InColorAndOpacity:Linear Color Structure=Select(Option 0:Linear Color Structure=(R=1,G=1,B=1,A=0), Option 1:Linear Color Structure=(R=0.0009,G=0.006,B=0.0395,A=1), Index=@n1.Selected) -> (ReturnValue:Linear Color Structure)
   )
   n2 = Get Border_0() -> (Border_0:Border Object Reference)
   n3 = BreakColor(InColor=Get BrushColor!(self:Border Object Reference=@n2) -> (BrushColor:Linear Color Structure))
   > SetBrushColor(
    self=@n2
    InBrushColor=Select(
     Option 0=MakeColor!(R:float=@n3, G:float=@n3.G, B:float=@n3.B, A:float=0) -> (ReturnValue:Linear Color Structure)
     Option 1=MakeColor(R=@n3, G=@n3.G, B=@n3.B, A=1)
     Index=@n1.Selected
    )
   )
  ]}]
variables [
 DefaultColor {type=SlateColor, default {colorUseRule=UseColor_Specified, specifiedColor=<rgba 1 1 1 1>}}
 SelectedColor {type=SlateColor, default {colorUseRule=UseColor_Specified, specifiedColor=<rgba 0.0009 0.006 0.0395 1>}}
]
widgets {
 info {inheritedWidgetCount=0, parentClass=&UserWidget, rootWidgetClass=&Border, widgetCount=5}
 widgets [
  Border_0 {
   widgetClassPath=&Border
   bIsVariable=true
   widget=Border_0
   properties {verticalAlignment=VAlign_Center, brushColor=<rgba 0.9823 0.9823 0.9823 1>, padding=<ltrb 16 0 0 0>}
  }
  HorizontalBox_0 {
   widgetClassPath=&HorizontalBox
   parent=Border_0
   widget=HorizontalBox_0
   slot {class=BorderSlot, properties {verticalAlignment=VAlign_Center, padding=<ltrb 16 0 0 0>}}
  }
  Image_60 {
   widgetClassPath=&Image
   bIsVariable=true
   parent=HorizontalBox_0
   widget=Image_60
   properties {
    brush {
     bIsDynamicallyLoaded=false
     drawAs=Image
     imageType=NoImage
     mirroring=NoMirror
     resourceName=None
     resourceObject=/Game/UI/Friends_UI/Social/Window_Arrow_16x
     tiling=NoTile
     imageSize=<16 16>
     margin=<ltrb 0 0 0 0>
     outlineSettings {
      bUseBrushTransparency=false
      roundingType=HalfHeightRadius
      width=0
      color {colorUseRule=UseColor_Specified, specifiedColor=<rgba 0 0 0 0>}
      cornerRadii=<0 0 0 0>
     }
     tintColor {colorUseRule=UseColor_Specified, specifiedColor=<rgba 1 1 1 1>}
     uVRegion {bIsValid=false, max=<0 0>, min=<0 0>}
    }
    colorAndOpacity=<rgba 0.0009 0.006 0.0395 1>
   }
   slot {class=HorizontalBoxSlot, properties {horizontalAlignment=HAlign_Left, verticalAlignment=VAlign_Center}}
  }
  SizeBox_1 {
   widgetClassPath=&SizeBox
   parent=HorizontalBox_0
   widget=SizeBox_1
   properties {bOverride_HeightOverride=true, bOverride_MinDesiredWidth=true, heightOverride=40, minDesiredWidth=400}
   slot {class=HorizontalBoxSlot, properties {verticalAlignment=VAlign_Center, size {sizeRule=Fill, value=1}}}
  }
  CommonTextBlock_34 {
   widgetClassPath=&CommonTextBlock
   bIsVariable=true
   parent=SizeBox_1
   widget=CommonTextBlock_34
   properties {
    flowDirectionPreference=LeftToRight
    scrollStyle=&TextScrollStyle-Base_C
    text=Text Block
    textTransformPolicy=ToUpper
    colorAndOpacity {colorUseRule=UseColor_Specified, specifiedColor=<rgba 0.0009 0.006 0.0395 1>}
    font {
     bForceMonospaced=false
     bMaterialIsStencil=false
     fontMaterial=null
     fontObject=&HeadingNow
     letterSpacing=0
     monospacedWidth=1
     size=16.5
     skewAmount=0
     typefaceFontName=75 Medium
     outlineSettings {
      bApplyOutlineToDropShadows=false
      bMiteredCorners=false
      bSeparateFillAlpha=false
      outlineMaterial=null
      outlineSize=0
      outlineColor=<rgba 0 0 0 1>
     }
    }
    shadowOffset=<0 0>
    shapedTextOptions {bOverride_TextFlowDirection=true, bOverride_TextShapingMethod=false, textFlowDirection=LeftToRight, textShapingMethod=Auto}
   }
   slot {class=SizeBoxSlot, properties {verticalAlignment=VAlign_Center, padding=<ltrb 0 0 16 0>}}
  }
 ]
}
