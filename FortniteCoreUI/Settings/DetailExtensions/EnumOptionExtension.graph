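# Settings detail extension widget that lists the choices of a discrete setting in the settings details panel.
# When a setting is assigned or changes, RebuildOptions casts it to FortSettingValueDiscrete and refills a DynamicEntryBox, marking the current option.
# If the split disabled visual is on and disabled options exist, it lists them unselected in a second box under a warning text giving the reasons.
# Otherwise that warning and second box are collapsed.
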WidgetBlueprint /FortniteCoreUI/Settings/DetailExtensions/EnumOptionExtension
graphs [
 EventGraph {nodes [
   n1 = Event OnSettingValueChanged!() -> (OutputDelegate, then, InSetting:Fort Setting Object Reference) {is=title K2Node_Event}
   > RebuildOptions!(execute, self, Setting:Object Reference=@n1.InSetting) -> (then)
   n2 = Event OnSettingAssigned!() -> (OutputDelegate, then, InSetting:Fort Setting Object Reference) {is=title K2Node_Event}
   > RebuildOptions(Setting=@n2.InSetting)
  ]}
 RebuildOptions {nodes [
   n1 = RebuildOptions!() -> (then, Setting:Object Reference) {is=title K2Node_FunctionEntry}
   > Reset!(
    execute
    self:Dynamic Entry Box Object Reference=Get EnumOptions() -> (EnumOptions:Dynamic Entry Box Object Reference)
    bDeleteWidgets:bool=false
   ) -> (then)
   > n2 = Cast To FortSettingValueDiscrete!(execute, Object=@n1) -> (then, CastFailed, AsFort Setting Value Discrete:Fort Setting Value Discrete Object Reference) {is=title K2Node_DynamicCast}
   > n3 = For Each Loop(Array:Array of Texts=GetDiscreteOptions!(self:Fort Setting Value Discrete Object Reference=@n2) -> (ReturnValue:Array of Texts)) -> (Array Element:Text)
   > n4 = Create Entry of Class!(
    execute
    self:Dynamic Entry Box Object Reference=Get EnumOptions()
    EntryClass:User Widget Class Reference=&/EnumOptionDetailsEntry.EnumOptionDetailsEntry_C
   ) -> (then, ReturnValue:Enum Option Details Entry Object Reference)
   > SetupEntry!(
    execute
    self:Enum Option Details Entry Object Reference=@n4
    Text:Text=@n3
    Selected:bool="Equal (Integer)"(A=@n3.Array Index, B=GetDiscreteOptionIndex!(self:Fort Setting Value Discrete Object Reference=@n2) -> (ReturnValue:int=0)) {is=title K2Node_CallFunction}
   ) -> (then)
   > Reset(self=Get EnumOptions_Disabled() -> (EnumOptions_Disabled:Dynamic Entry Box Object Reference), bDeleteWidgets=false)
   > n5 = UseSplitDisabledOptionsVisual!(execute, self:Fort Setting Detail Extension Object Reference) -> (then, ReturnValue:bool)
   n6 = GetDisabledOptions!(self:Fort Setting Value Discrete Object Reference=@n2) -> (ReturnValue:Array of Texts)
   > n7 = Branch(Condition=OR Boolean(A=NOT Boolean(A=@n5), B=Is Empty!(TargetArray:Array of Texts=@n6) -> (ReturnValue:bool)))
   > SetVisibility(self=Get RichText_DisabledDetails() -> (RichText_DisabledDetails:Common Rich Text Block Object Reference), InVisibility=Collapsed)
   > SetVisibility(self=Get EnumOptions_Disabled(), InVisibility=Collapsed)
   n8 = For Each Loop(Exec=@n7.else, Array=@n6)
   > n9 = Create Entry of Class(self=Get EnumOptions_Disabled(), EntryClass=&EnumOptionDetailsEntry_C)
   > SetupEntry(self=@n9, Text=@n8, Selected=false)
   > SetVisibility(self=Get RichText_DisabledDetails(), InVisibility=Visible)
   > SetVisibility(self=Get EnumOptions_Disabled(), InVisibility=Visible)
   > SetText(
    self=Get RichText_DisabledDetails()
    InText=GetOptionsDisabledText!(
     self:Fort Setting Detail Extension Object Reference
     Reasons:Array of Texts=GetDisabledOptionsReasons!(self:Fort Setting Value Discrete Object Reference=@n2) -> (ReturnValue:Array of Texts)
    ) -> (ReturnValue:Text)
   )
  ]}
]
variables []
widgets {
 info {inheritedWidgetCount=0, parentClass=/Script/FortniteUI.FortSettingDetailExtension, rootWidgetClass=&VerticalBox, widgetCount=5}
 widgets [
  VerticalBox_0 {widgetClassPath=&VerticalBox, widget=VerticalBox_0}
  EnumOptions {
   widgetClassPath=&DynamicEntryBox
   bIsVariable=true
   parent=VerticalBox_0
   widget=EnumOptions
   properties {
    entryBoxType=Vertical
    entryHorizontalAlignment=HAlign_Left
    entryWidgetClass=&EnumOptionDetailsEntry_C
   }
   slot {class=VerticalBoxSlot}
  }
  RichText_DisabledDetails {
   widgetClassPath=&CommonRichTextBlock
   bIsVariable=true
   parent=VerticalBox_0
   widget=RichText_DisabledDetails
   properties {
    autoWrapText=true
    mobileTextBlockScale=1.34
    text="<Icon.Warning></> DISABLED BY PARENTAL CONTROLS"
    textStyleSet=/FortniteCoreUI/Settings/SettingsDescriptionWarningStyes
    defaultTextStyleOverride {
     overflowPolicy=Clip
     transformPolicy=None
     colorAndOpacity {colorUseRule=UseColor_Specified, specifiedColor=<rgba 1 1 1 1>}
     font {
      bForceMonospaced=false
      bMaterialIsStencil=false
      fontMaterial=null
      fontObject=&NotoSans
      letterSpacing=0
      monospacedWidth=1
      size=24
      skewAmount=0
      typefaceFontName=NotoSans-Regular
      outlineSettings {
       bApplyOutlineToDropShadows=false
       bMiteredCorners=false
       bSeparateFillAlpha=false
       outlineMaterial=null
       outlineSize=0
       outlineColor=<rgba 0 0 0 1>
      }
     }
     highlightColor {colorUseRule=UseColor_Specified, specifiedColor=<rgba 0 0 0 1>}
     highlightShape {
      bIsDynamicallyLoaded=false
      drawAs=Image
      imageType=NoImage
      mirroring=NoMirror
      resourceName=None
      resourceObject=null
      tiling=NoTile
      imageSize=<32 32>
      margin=<ltrb 0 0 0 0>
      outlineSettings {
       bUseBrushTransparency=false
       roundingType=HalfHeightRadius
       width=0
       color {colorUseRule=UseColor_Specified, specifiedColor=<rgba 0 0 0 0>}
       cornerRadii=<0 0 0 0>
      }
      tintColor {colorUseRule=UseColor_Specified, specifiedColor=<rgba 1 1 1 1>}
      uVRegion {bIsValid=false, max=<0 0>, min=<0 0>}
     }
     selectedBackgroundColor {colorUseRule=UseColor_Foreground, specifiedColor=<rgba 1 0 1 1>}
     shadowColorAndOpacity=<rgba 0 0 0 1>
     shadowOffset=<0 0>
     strikeBrush {
      bIsDynamicallyLoaded=false
      drawAs=Image
      imageType=NoImage
      mirroring=NoMirror
      resourceName=None
      resourceObject=null
      tiling=NoTile
      imageSize=<32 32>
      margin=<ltrb 0 0 0 0>
      outlineSettings {
       bUseBrushTransparency=false
       roundingType=HalfHeightRadius
       width=0
       color {colorUseRule=UseColor_Specified, specifiedColor=<rgba 0 0 0 0>}
       cornerRadii=<0 0 0 0>
      }
      tintColor {colorUseRule=UseColor_Specified, specifiedColor=<rgba 1 1 1 1>}
      uVRegion {bIsValid=false, max=<0 0>, min=<0 0>}
     }
     underlineBrush {
      bIsDynamicallyLoaded=false
      drawAs=Image
      imageType=NoImage
      mirroring=NoMirror
      resourceName=None
      resourceObject=null
      tiling=NoTile
      imageSize=<32 32>
      margin=<ltrb 0 0 0 0>
      outlineSettings {
       bUseBrushTransparency=false
       roundingType=HalfHeightRadius
       width=0
       color {colorUseRule=UseColor_Specified, specifiedColor=<rgba 0 0 0 0>}
       cornerRadii=<0 0 0 0>
      }
      tintColor {colorUseRule=UseColor_Specified, specifiedColor=<rgba 1 1 1 1>}
      uVRegion {bIsValid=false, max=<0 0>, min=<0 0>}
     }
    }
   }
   slot {class=VerticalBoxSlot, properties {padding=<ltrb 0 20 0 0>}}
  }
  Spacer_51 {widgetClassPath=&Spacer, parent=VerticalBox_0, widget=Spacer_51, properties {size=<1 10>}, slot {class=VerticalBoxSlot}}
  EnumOptions_Disabled {
   widgetClassPath=&DynamicEntryBox
   bIsVariable=true
   parent=VerticalBox_0
   widget=EnumOptions_Disabled
   properties {
    entryBoxType=Vertical
    entryHorizontalAlignment=HAlign_Left
    entryWidgetClass=&EnumOptionDetailsEntry_C
   }
   slot {class=VerticalBoxSlot}
  }
 ]
}
