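# Settings detail extension widget that displays a row of five preferred item slots, built on the native FortSettingsPreferredItemSlotsExtension class.
# Each fixed-size slot overlays an empty-slot material, a slightly enlarged highlight frame and a padded slot-type icon inside the parent's HorizontalBox_ItemSlotContainer.
# PreConstruct gathers the icons and highlights into the SlotItemIconImages and HighlightImages arrays; the Blueprint has no other logic, as Tick is empty.
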WidgetBlueprint /FortniteCoreUI/Settings/DetailExtensions/PreferredItemSlotsExtension
graphs [EventGraph {nodes [
   Event PreConstruct()
   > Set SlotItemIconImages(SlotItemIconImages:Array of Image Object References=Make Array!(
     "[0]":Image Object Reference=Get SlotTypeImage() -> (SlotTypeImage:Image Object Reference)
     "[1]":Image Object Reference=Get SlotTypeImage_1() -> (SlotTypeImage_1:Image Object Reference)
     "[2]":Image Object Reference=Get SlotTypeImage_2() -> (SlotTypeImage_2:Image Object Reference)
     "[3]":Image Object Reference=Get SlotTypeImage_3() -> (SlotTypeImage_3:Image Object Reference)
     "[4]":Image Object Reference=Get SlotTypeImage_4() -> (SlotTypeImage_4:Image Object Reference)
    ) -> (Array:Array of Image Object References)) -> (Output_Get:Array of Image Object References)
   > Set HighlightImages(HighlightImages:Array of Image Object References=Make Array(
     "[0]"=Get HoverImage() -> (HoverImage:Image Object Reference)
     "[1]"=Get HoverImage_1() -> (HoverImage_1:Image Object Reference)
     "[2]"=Get HoverImage_2() -> (HoverImage_2:Image Object Reference)
     "[3]"=Get HoverImage_3() -> (HoverImage_3:Image Object Reference)
     "[4]"=Get HoverImage_4() -> (HoverImage_4:Image Object Reference)
    )) -> (Output_Get:Array of Image Object References)
   Event Tick()
  ]}]
variables []
widgets {
 info {
  inheritedWidgetCount=1
  parentClass=/Script/FortniteUI.FortSettingsPreferredItemSlotsExtension
  rootWidgetClass=&VerticalBox
  widgetCount=33
 }
 widgets [
  VerticalBox_0 {widgetClassPath=&VerticalBox, widget=VerticalBox_0}
  HorizontalBox_ItemSlotContainer {
   widgetClassPath=&HorizontalBox
   bInherited=true
   bIsVariable=true
   parent=VerticalBox_0
   widget=HorizontalBox_ItemSlotContainer
   slot {class=VerticalBoxSlot, properties {padding=<ltrb 15 0 0 0>}}
  }
  Slot1 {
   widgetClassPath=&SizeBox
   parent=HorizontalBox_ItemSlotContainer
   widget=Slot1
   properties {bOverride_HeightOverride=true, bOverride_WidthOverride=true, heightOverride=122, widthOverride=122}
   slot {class=HorizontalBoxSlot, properties {horizontalAlignment=HAlign_Left, verticalAlignment=VAlign_Top, padding=<ltrb 0 0 15 0>}}
  }
  OverlayContent {widgetClassPath=&Overlay, parent=Slot1, widget=OverlayContent, slot {class=SizeBoxSlot}}
  EmptyImage {
   widgetClassPath=&Image
   parent=OverlayContent
   widget=EmptyImage
   properties {
    visibility=SelfHitTestInvisible
    brush {
     bIsDynamicallyLoaded=false
     drawAs=Box
     imageType=NoImage
     mirroring=NoMirror
     resourceName=None
     resourceObject=/Game/Athena/HUD/Inventory/Art/M_UI_Inventory_LoadoutItem_EmptySlot
     tiling=NoTile
     imageSize=<64 64>
     margin=<ltrb 0.5 0.5 0.5 0.5>
     outlineSettings {
      bUseBrushTransparency=false
      roundingType=HalfHeightRadius
      width=0
      color {colorUseRule=UseColor_Specified, specifiedColor=<rgba 0 0 0 0>}
      cornerRadii=<0 0 0 0>
     }
     tintColor {colorUseRule=UseColor_Specified, specifiedColor=<rgba 1 1 1 1>}
     uVRegion {bIsValid=false, max=<0 0>, min=<0 0>}
    }
   }
   slot {class=OverlaySlot, properties {horizontalAlignment=HAlign_Fill, verticalAlignment=VAlign_Fill}}
  }
  HoverImage {
   widgetClassPath=&Image
   bIsVariable=true
   parent=OverlayContent
   widget=HoverImage
   properties {
    visibility=SelfHitTestInvisible
    brush {
     bIsDynamicallyLoaded=false
     drawAs=Box
     imageType=NoImage
     mirroring=NoMirror
     resourceName=None
     resourceObject=/Game/UI/InGame/Quickbar/M_UI_FramePreview_Inst
     tiling=NoTile
     imageSize=<64 64>
     margin=<ltrb 0.5 0.5 0.5 0.5>
     outlineSettings {
      bUseBrushTransparency=false
      roundingType=HalfHeightRadius
      width=0
      color {colorUseRule=UseColor_Specified, specifiedColor=<rgba 0 0 0 0>}
      cornerRadii=<0 0 0 0>
     }
     tintColor {colorUseRule=UseColor_Specified, specifiedColor=<rgba 1 1 1 1>}
     uVRegion {bIsValid=false, max=<0 0>, min=<0 0>}
    }
    renderTransform {angle=0, scale=<1.1 1.1>, shear=<0 0>, translation=<0 0>}
   }
   slot {class=OverlaySlot, properties {horizontalAlignment=HAlign_Fill, verticalAlignment=VAlign_Fill}}
  }
  Border_153 {
   widgetClassPath=&Border
   parent=OverlayContent
   widget=Border_153
   properties {
    background {
     bIsDynamicallyLoaded=false
     drawAs=NoDrawType
     imageType=NoImage
     mirroring=NoMirror
     resourceName=None
     resourceObject=null
     tiling=NoTile
     imageSize=<32 32>
     margin=<ltrb 0 0 0 0>
     outlineSettings {
      bUseBrushTransparency=false
      roundingType=HalfHeightRadius
      width=0
      color {colorUseRule=UseColor_Specified, specifiedColor=<rgba 0 0 0 0>}
      cornerRadii=<0 0 0 0>
     }
     tintColor {colorUseRule=UseColor_Specified, specifiedColor=<rgba 1 1 1 1>}
     uVRegion {bIsValid=false, max=<0 0>, min=<0 0>}
    }
    padding=<ltrb 10 10 10 10>
   }
   slot {class=OverlaySlot, properties {horizontalAlignment=HAlign_Fill, verticalAlignment=VAlign_Fill}}
  }
  SlotTypeImage {
   widgetClassPath=&Image
   bIsVariable=true
   parent=Border_153
   widget=SlotTypeImage
   properties {brush {
     bIsDynamicallyLoaded=false
     drawAs=Image
     imageType=NoImage
     mirroring=NoMirror
     resourceName=None
     resourceObject=null
     tiling=NoTile
     imageSize=<128 128>
     margin=<ltrb 0 0 0 0>
     outlineSettings {
      bUseBrushTransparency=false
      roundingType=HalfHeightRadius
      width=0
      color {colorUseRule=UseColor_Specified, specifiedColor=<rgba 0 0 0 0>}
      cornerRadii=<0 0 0 0>
     }
     tintColor {colorUseRule=UseColor_Specified, specifiedColor=<rgba 1 1 1 1>}
     uVRegion {bIsValid=false, max=<0 0>, min=<0 0>}
    }}
   slot {class=BorderSlot, properties {padding=<ltrb 10 10 10 10>}}
  }
  Slot2 {
   widgetClassPath=&SizeBox
   parent=HorizontalBox_ItemSlotContainer
   widget=Slot2
   properties {bOverride_HeightOverride=true, bOverride_WidthOverride=true, heightOverride=122, widthOverride=122}
   slot {class=HorizontalBoxSlot, properties {horizontalAlignment=HAlign_Left, verticalAlignment=VAlign_Top, padding=<ltrb 0 0 15 0>}}
  }
  OverlayContent_1 {widgetClassPath=&Overlay, parent=Slot2, widget=OverlayContent_1, slot {class=SizeBoxSlot}}
  HoverImage_1 {
   widgetClassPath=&Image
   bIsVariable=true
   parent=OverlayContent_1
   widget=HoverImage_1
   properties {
    visibility=SelfHitTestInvisible
    brush {
     bIsDynamicallyLoaded=false
     drawAs=Box
     imageType=NoImage
     mirroring=NoMirror
     resourceName=None
     resourceObject=&M_UI_FramePreview_Inst
     tiling=NoTile
     imageSize=<64 64>
     margin=<ltrb 0.5 0.5 0.5 0.5>
     outlineSettings {
      bUseBrushTransparency=false
      roundingType=HalfHeightRadius
      width=0
      color {colorUseRule=UseColor_Specified, specifiedColor=<rgba 0 0 0 0>}
      cornerRadii=<0 0 0 0>
     }
     tintColor {colorUseRule=UseColor_Specified, specifiedColor=<rgba 1 1 1 1>}
     uVRegion {bIsValid=false, max=<0 0>, min=<0 0>}
    }
    renderTransform {angle=0, scale=<1.1 1.1>, shear=<0 0>, translation=<0 0>}
   }
   slot {class=OverlaySlot, properties {horizontalAlignment=HAlign_Fill, verticalAlignment=VAlign_Fill}}
  }
  EmptySlotImage {
   widgetClassPath=&Image
   parent=OverlayContent_1
   widget=EmptySlotImage
   properties {
    visibility=SelfHitTestInvisible
    brush {
     bIsDynamicallyLoaded=false
     drawAs=Box
     imageType=NoImage
     mirroring=NoMirror
     resourceName=None
     resourceObject=&M_UI_Inventory_LoadoutItem_EmptySlot
     tiling=NoTile
     imageSize=<64 64>
     margin=<ltrb 0.5 0.5 0.5 0.5>
     outlineSettings {
      bUseBrushTransparency=false
      roundingType=HalfHeightRadius
      width=0
      color {colorUseRule=UseColor_Specified, specifiedColor=<rgba 0 0 0 0>}
      cornerRadii=<0 0 0 0>
     }
     tintColor {colorUseRule=UseColor_Specified, specifiedColor=<rgba 1 1 1 1>}
     uVRegion {bIsValid=false, max=<0 0>, min=<0 0>}
    }
   }
   slot {class=OverlaySlot, properties {horizontalAlignment=HAlign_Fill, verticalAlignment=VAlign_Fill}}
  }
  Border {
   widgetClassPath=&Border
   parent=OverlayContent_1
   widget=Border
   properties {
    background {
     bIsDynamicallyLoaded=false
     drawAs=NoDrawType
     imageType=NoImage
     mirroring=NoMirror
     resourceName=None
     resourceObject=null
     tiling=NoTile
     imageSize=<32 32>
     margin=<ltrb 0 0 0 0>
     outlineSettings {
      bUseBrushTransparency=false
      roundingType=HalfHeightRadius
      width=0
      color {colorUseRule=UseColor_Specified, specifiedColor=<rgba 0 0 0 0>}
      cornerRadii=<0 0 0 0>
     }
     tintColor {colorUseRule=UseColor_Specified, specifiedColor=<rgba 1 1 1 1>}
     uVRegion {bIsValid=false, max=<0 0>, min=<0 0>}
    }
    padding=<ltrb 10 10 10 10>
   }
   slot {class=OverlaySlot, properties {horizontalAlignment=HAlign_Fill, verticalAlignment=VAlign_Fill}}
  }
  SlotTypeImage_1 {
   widgetClassPath=&Image
   bIsVariable=true
   parent=Border
   widget=SlotTypeImage_1
   slot {class=BorderSlot, properties {padding=<ltrb 10 10 10 10>}}
  }
  Slot3 {
   widgetClassPath=&SizeBox
   parent=HorizontalBox_ItemSlotContainer
   widget=Slot3
   properties {bOverride_HeightOverride=true, bOverride_WidthOverride=true, heightOverride=122, widthOverride=122}
   slot {class=HorizontalBoxSlot, properties {horizontalAlignment=HAlign_Left, verticalAlignment=VAlign_Top, padding=<ltrb 0 0 15 0>}}
  }
  OverlayContent_2 {widgetClassPath=&Overlay, parent=Slot3, widget=OverlayContent_2, slot {class=SizeBoxSlot}}
  HoverImage_2 {
   widgetClassPath=&Image
   bIsVariable=true
   parent=OverlayContent_2
   widget=HoverImage_2
   properties {
    visibility=SelfHitTestInvisible
    brush {
     bIsDynamicallyLoaded=false
     drawAs=Box
     imageType=NoImage
     mirroring=NoMirror
     resourceName=None
     resourceObject=&M_UI_FramePreview_Inst
     tiling=NoTile
     imageSize=<64 64>
     margin=<ltrb 0.5 0.5 0.5 0.5>
     outlineSettings {
      bUseBrushTransparency=false
      roundingType=HalfHeightRadius
      width=0
      color {colorUseRule=UseColor_Specified, specifiedColor=<rgba 0 0 0 0>}
      cornerRadii=<0 0 0 0>
     }
     tintColor {colorUseRule=UseColor_Specified, specifiedColor=<rgba 1 1 1 1>}
     uVRegion {bIsValid=false, max=<0 0>, min=<0 0>}
    }
    renderTransform {angle=0, scale=<1.1 1.1>, shear=<0 0>, translation=<0 0>}
   }
   slot {class=OverlaySlot, properties {horizontalAlignment=HAlign_Fill, verticalAlignment=VAlign_Fill}}
  }
  EmptyImage_2 {
   widgetClassPath=&Image
   parent=OverlayContent_2
   widget=EmptyImage_2
   properties {
    visibility=SelfHitTestInvisible
    brush {
     bIsDynamicallyLoaded=false
     drawAs=Box
     imageType=NoImage
     mirroring=NoMirror
     resourceName=None
     resourceObject=&M_UI_Inventory_LoadoutItem_EmptySlot
     tiling=NoTile
     imageSize=<64 64>
     margin=<ltrb 0.5 0.5 0.5 0.5>
     outlineSettings {
      bUseBrushTransparency=false
      roundingType=HalfHeightRadius
      width=0
      color {colorUseRule=UseColor_Specified, specifiedColor=<rgba 0 0 0 0>}
      cornerRadii=<0 0 0 0>
     }
     tintColor {colorUseRule=UseColor_Specified, specifiedColor=<rgba 1 1 1 1>}
     uVRegion {bIsValid=false, max=<0 0>, min=<0 0>}
    }
   }
   slot {class=OverlaySlot, properties {horizontalAlignment=HAlign_Fill, verticalAlignment=VAlign_Fill}}
  }
  Border_1 {
   widgetClassPath=&Border
   parent=OverlayContent_2
   widget=Border_1
   properties {
    background {
     bIsDynamicallyLoaded=false
     drawAs=NoDrawType
     imageType=NoImage
     mirroring=NoMirror
     resourceName=None
     resourceObject=null
     tiling=NoTile
     imageSize=<32 32>
     margin=<ltrb 0 0 0 0>
     outlineSettings {
      bUseBrushTransparency=false
      roundingType=HalfHeightRadius
      width=0
      color {colorUseRule=UseColor_Specified, specifiedColor=<rgba 0 0 0 0>}
      cornerRadii=<0 0 0 0>
     }
     tintColor {colorUseRule=UseColor_Specified, specifiedColor=<rgba 1 1 1 1>}
     uVRegion {bIsValid=false, max=<0 0>, min=<0 0>}
    }
    padding=<ltrb 10 10 10 10>
   }
   slot {class=OverlaySlot, properties {horizontalAlignment=HAlign_Fill, verticalAlignment=VAlign_Fill}}
  }
  SlotTypeImage_2 {
   widgetClassPath=&Image
   bIsVariable=true
   parent=Border_1
   widget=SlotTypeImage_2
   slot {class=BorderSlot, properties {padding=<ltrb 10 10 10 10>}}
  }
  Slot4 {
   widgetClassPath=&SizeBox
   parent=HorizontalBox_ItemSlotContainer
   widget=Slot4
   properties {bOverride_HeightOverride=true, bOverride_WidthOverride=true, heightOverride=122, widthOverride=122}
   slot {class=HorizontalBoxSlot, properties {horizontalAlignment=HAlign_Left, verticalAlignment=VAlign_Top, padding=<ltrb 0 0 15 0>}}
  }
  OverlayContent_3 {widgetClassPath=&Overlay, parent=Slot4, widget=OverlayContent_3, slot {class=SizeBoxSlot}}
  HoverImage_3 {
   widgetClassPath=&Image
   bIsVariable=true
   parent=OverlayContent_3
   widget=HoverImage_3
   properties {
    visibility=SelfHitTestInvisible
    brush {
     bIsDynamicallyLoaded=false
     drawAs=Box
     imageType=NoImage
     mirroring=NoMirror
     resourceName=None
     resourceObject=&M_UI_FramePreview_Inst
     tiling=NoTile
     imageSize=<64 64>
     margin=<ltrb 0.5 0.5 0.5 0.5>
     outlineSettings {
      bUseBrushTransparency=false
      roundingType=HalfHeightRadius
      width=0
      color {colorUseRule=UseColor_Specified, specifiedColor=<rgba 0 0 0 0>}
      cornerRadii=<0 0 0 0>
     }
     tintColor {colorUseRule=UseColor_Specified, specifiedColor=<rgba 1 1 1 1>}
     uVRegion {bIsValid=false, max=<0 0>, min=<0 0>}
    }
    renderTransform {angle=0, scale=<1.1 1.1>, shear=<0 0>, translation=<0 0>}
   }
   slot {class=OverlaySlot, properties {horizontalAlignment=HAlign_Fill, verticalAlignment=VAlign_Fill}}
  }
  EmptyImage_3 {
   widgetClassPath=&Image
   parent=OverlayContent_3
   widget=EmptyImage_3
   properties {
    visibility=SelfHitTestInvisible
    brush {
     bIsDynamicallyLoaded=false
     drawAs=Box
     imageType=NoImage
     mirroring=NoMirror
     resourceName=None
     resourceObject=&M_UI_Inventory_LoadoutItem_EmptySlot
     tiling=NoTile
     imageSize=<64 64>
     margin=<ltrb 0.5 0.5 0.5 0.5>
     outlineSettings {
      bUseBrushTransparency=false
      roundingType=HalfHeightRadius
      width=0
      color {colorUseRule=UseColor_Specified, specifiedColor=<rgba 0 0 0 0>}
      cornerRadii=<0 0 0 0>
     }
     tintColor {colorUseRule=UseColor_Specified, specifiedColor=<rgba 1 1 1 1>}
     uVRegion {bIsValid=false, max=<0 0>, min=<0 0>}
    }
   }
   slot {class=OverlaySlot, properties {horizontalAlignment=HAlign_Fill, verticalAlignment=VAlign_Fill}}
  }
  Border_2 {
   widgetClassPath=&Border
   parent=OverlayContent_3
   widget=Border_2
   properties {
    background {
     bIsDynamicallyLoaded=false
     drawAs=NoDrawType
     imageType=NoImage
     mirroring=NoMirror
     resourceName=None
     resourceObject=null
     tiling=NoTile
     imageSize=<32 32>
     margin=<ltrb 0 0 0 0>
     outlineSettings {
      bUseBrushTransparency=false
      roundingType=HalfHeightRadius
      width=0
      color {colorUseRule=UseColor_Specified, specifiedColor=<rgba 0 0 0 0>}
      cornerRadii=<0 0 0 0>
     }
     tintColor {colorUseRule=UseColor_Specified, specifiedColor=<rgba 1 1 1 1>}
     uVRegion {bIsValid=false, max=<0 0>, min=<0 0>}
    }
    padding=<ltrb 10 10 10 10>
   }
   slot {class=OverlaySlot, properties {horizontalAlignment=HAlign_Fill, verticalAlignment=VAlign_Fill}}
  }
  SlotTypeImage_3 {
   widgetClassPath=&Image
   bIsVariable=true
   parent=Border_2
   widget=SlotTypeImage_3
   slot {class=BorderSlot, properties {padding=<ltrb 10 10 10 10>}}
  }
  Slot5 {
   widgetClassPath=&SizeBox
   parent=HorizontalBox_ItemSlotContainer
   widget=Slot5
   properties {bOverride_HeightOverride=true, bOverride_WidthOverride=true, heightOverride=122, widthOverride=122}
   slot {class=HorizontalBoxSlot, properties {horizontalAlignment=HAlign_Left, verticalAlignment=VAlign_Top}}
  }
  OverlayContent_4 {widgetClassPath=&Overlay, parent=Slot5, widget=OverlayContent_4, slot {class=SizeBoxSlot}}
  HoverImage_4 {
   widgetClassPath=&Image
   bIsVariable=true
   parent=OverlayContent_4
   widget=HoverImage_4
   properties {
    visibility=SelfHitTestInvisible
    brush {
     bIsDynamicallyLoaded=false
     drawAs=Box
     imageType=NoImage
     mirroring=NoMirror
     resourceName=None
     resourceObject=&M_UI_FramePreview_Inst
     tiling=NoTile
     imageSize=<64 64>
     margin=<ltrb 0.5 0.5 0.5 0.5>
     outlineSettings {
      bUseBrushTransparency=false
      roundingType=HalfHeightRadius
      width=0
      color {colorUseRule=UseColor_Specified, specifiedColor=<rgba 0 0 0 0>}
      cornerRadii=<0 0 0 0>
     }
     tintColor {colorUseRule=UseColor_Specified, specifiedColor=<rgba 1 1 1 1>}
     uVRegion {bIsValid=false, max=<0 0>, min=<0 0>}
    }
    renderTransform {angle=0, scale=<1.1 1.1>, shear=<0 0>, translation=<0 0>}
   }
   slot {class=OverlaySlot, properties {horizontalAlignment=HAlign_Fill, verticalAlignment=VAlign_Fill}}
  }
  EmptyImage_4 {
   widgetClassPath=&Image
   parent=OverlayContent_4
   widget=EmptyImage_4
   properties {
    visibility=SelfHitTestInvisible
    brush {
     bIsDynamicallyLoaded=false
     drawAs=Box
     imageType=NoImage
     mirroring=NoMirror
     resourceName=None
     resourceObject=&M_UI_Inventory_LoadoutItem_EmptySlot
     tiling=NoTile
     imageSize=<64 64>
     margin=<ltrb 0.5 0.5 0.5 0.5>
     outlineSettings {
      bUseBrushTransparency=false
      roundingType=HalfHeightRadius
      width=0
      color {colorUseRule=UseColor_Specified, specifiedColor=<rgba 0 0 0 0>}
      cornerRadii=<0 0 0 0>
     }
     tintColor {colorUseRule=UseColor_Specified, specifiedColor=<rgba 1 1 1 1>}
     uVRegion {bIsValid=false, max=<0 0>, min=<0 0>}
    }
   }
   slot {class=OverlaySlot, properties {horizontalAlignment=HAlign_Fill, verticalAlignment=VAlign_Fill}}
  }
  Border_3 {
   widgetClassPath=&Border
   parent=OverlayContent_4
   widget=Border_3
   properties {
    background {
     bIsDynamicallyLoaded=false
     drawAs=NoDrawType
     imageType=NoImage
     mirroring=NoMirror
     resourceName=None
     resourceObject=null
     tiling=NoTile
     imageSize=<32 32>
     margin=<ltrb 0 0 0 0>
     outlineSettings {
      bUseBrushTransparency=false
      roundingType=HalfHeightRadius
      width=0
      color {colorUseRule=UseColor_Specified, specifiedColor=<rgba 0 0 0 0>}
      cornerRadii=<0 0 0 0>
     }
     tintColor {colorUseRule=UseColor_Specified, specifiedColor=<rgba 1 1 1 1>}
     uVRegion {bIsValid=false, max=<0 0>, min=<0 0>}
    }
    padding=<ltrb 10 10 10 10>
   }
   slot {class=OverlaySlot, properties {horizontalAlignment=HAlign_Fill, verticalAlignment=VAlign_Fill}}
  }
  SlotTypeImage_4 {
   widgetClassPath=&Image
   bIsVariable=true
   parent=Border_3
   widget=SlotTypeImage_4
   slot {class=BorderSlot, properties {padding=<ltrb 10 10 10 10>}}
  }
  Spacer_39 {widgetClassPath=&Spacer, parent=VerticalBox_0, widget=Spacer_39, properties {size=<1 50>}, slot {class=VerticalBoxSlot}}
 ]
}
