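# Settings detail extension widget that shows the currently selected checkout method as a rich text line in the settings details.
# When a setting is assigned or its value changes, it reads the selected app store and formats the label into its single Common Rich Text Block.
# A Select gives Epic and Samsung friendly names, uses the enum name for most other stores, and shows "Not Set" when the lookup fails.
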WidgetBlueprint /FortniteCoreUI/Settings/DetailExtensions/SelectCheckoutExtension
graphs [EventGraph {nodes [
   n1 = Event OnSettingValueChanged!() -> (OutputDelegate, then, InSetting:Fort Setting Object Reference) {is=title K2Node_Event}
   n2 = GetSelectedAppStore!() -> (OutAppStore:EAppStore Enum=DebugStore, ReturnValue:bool)
   n3 = Enum to String!(Enumerator:EAppStore Enum=@n2) -> (ReturnValue:String)
   n4 = Select!(
    EpicPurchasingService:String=Epic
    IOSAppStore:String=@n3
    WeGameStore:String=@n3
    GooglePlayAppStore:String=@n3
    KindleStore:String=@n3
    PlayStation4Store:String=@n3
    XboxLiveStore:String=@n3
    NintendoEShop:String=@n3
    SamsungGalaxyAppStore:String=Samsung
    MicrosoftStore:String=@n3
    PlayStation5Store:String=@n3
    OneStore:String
    Index:EAppStore Enum=@n2
   ) -> (ReturnValue:String)
   Set Text(
    execute=[@n1.then, @n5.then]
    Text=Format Text!(
     Format=MakeLiteralText(Value="LOCTEXT(\"<strong>Currently Selected Method</> : {0}\")")
     0:String=Select!(Option 0:String=Not Set, Option 1:String=@n4, Index:bool=@n2.ReturnValue) -> (ReturnValue)
    ) -> (Result)
    self:Rich Text Block Object Reference=Get RichTextDescription() -> (RichTextDescription:Common Rich Text Block Object Reference)
   )
   n5 = Event OnSettingAssigned!() -> (OutputDelegate, then, InSetting:Fort Setting Object Reference) {is=title K2Node_Event}
  ]}]
variables []
widgets {
 info {
  inheritedWidgetCount=0
  parentClass=/Script/FortniteUI.FortSettingDetailExtension
  rootWidgetClass=&CommonRichTextBlock
  widgetCount=1
 }
 widgets [RichTextDescription {
   widgetClassPath=&CommonRichTextBlock
   bIsVariable=true
   widget=RichTextDescription
   properties {
    autoWrapText=true
    lineHeightPercentage=1.3
    mobileTextBlockScale=1.34
    text="<strong>Currently Selected Method</> : {0}"
    textStyleSet=/FortniteCoreUI/Settings/SettingsDetails/SettingsDescriptionStyles
    defaultTextStyleOverride {
     overflowPolicy=Clip
     transformPolicy=None
     colorAndOpacity {colorUseRule=UseColor_Specified, specifiedColor=<rgba 1 1 1 1>}
     font {
      bForceMonospaced=false
      bMaterialIsStencil=false
      fontMaterial=null
      fontObject=&NotoSans
      letterSpacing=0
      monospacedWidth=1
      size=24
      skewAmount=0
      typefaceFontName=NotoSans-Regular
      outlineSettings {
       bApplyOutlineToDropShadows=false
       bMiteredCorners=false
       bSeparateFillAlpha=false
       outlineMaterial=null
       outlineSize=0
       outlineColor=<rgba 0 0 0 1>
      }
     }
     highlightColor {colorUseRule=UseColor_Specified, specifiedColor=<rgba 0 0 0 1>}
     highlightShape {
      bIsDynamicallyLoaded=false
      drawAs=Image
      imageType=NoImage
      mirroring=NoMirror
      resourceName=None
      resourceObject=null
      tiling=NoTile
      imageSize=<32 32>
      margin=<ltrb 0 0 0 0>
      outlineSettings {
       bUseBrushTransparency=false
       roundingType=HalfHeightRadius
       width=0
       color {colorUseRule=UseColor_Specified, specifiedColor=<rgba 0 0 0 0>}
       cornerRadii=<0 0 0 0>
      }
      tintColor {colorUseRule=UseColor_Specified, specifiedColor=<rgba 1 1 1 1>}
      uVRegion {bIsValid=false, max=<0 0>, min=<0 0>}
     }
     selectedBackgroundColor {colorUseRule=UseColor_Foreground, specifiedColor=<rgba 1 0 1 1>}
     shadowColorAndOpacity=<rgba 0 0 0 1>
     shadowOffset=<0 0>
     strikeBrush {
      bIsDynamicallyLoaded=false
      drawAs=Image
      imageType=NoImage
      mirroring=NoMirror
      resourceName=None
      resourceObject=null
      tiling=NoTile
      imageSize=<32 32>
      margin=<ltrb 0 0 0 0>
      outlineSettings {
       bUseBrushTransparency=false
       roundingType=HalfHeightRadius
       width=0
       color {colorUseRule=UseColor_Specified, specifiedColor=<rgba 0 0 0 0>}
       cornerRadii=<0 0 0 0>
      }
      tintColor {colorUseRule=UseColor_Specified, specifiedColor=<rgba 1 1 1 1>}
      uVRegion {bIsValid=false, max=<0 0>, min=<0 0>}
     }
     underlineBrush {
      bIsDynamicallyLoaded=false
      drawAs=Image
      imageType=NoImage
      mirroring=NoMirror
      resourceName=None
      resourceObject=null
      tiling=NoTile
      imageSize=<32 32>
      margin=<ltrb 0 0 0 0>
      outlineSettings {
       bUseBrushTransparency=false
       roundingType=HalfHeightRadius
       width=0
       color {colorUseRule=UseColor_Specified, specifiedColor=<rgba 0 0 0 0>}
       cornerRadii=<0 0 0 0>
      }
      tintColor {colorUseRule=UseColor_Specified, specifiedColor=<rgba 1 1 1 1>}
      uVRegion {bIsValid=false, max=<0 0>, min=<0 0>}
     }
    }
   }
  }]
}
