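# SubtitleExtension is a settings detail extension widget, based on FortSettingDetailExtension, that shows a live sample of subtitle styling.
# A ScaleBox scaled by safe zone fills horizontally with a SubtitleDisplay running in preview mode.
# The display takes its styling from the shared SubtitleOptions asset and shows a fixed SEE-Bot sample line with text wrapping.
# The event graph is empty, so the widget adds no Blueprint logic.
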WidgetBlueprint /FortniteCoreUI/Settings/DetailExtensions/SubtitleExtension
graphs [EventGraph {links [], nodes []}]
variables []
widgets {
 info {inheritedWidgetCount=0, parentClass=/Script/FortniteUI.FortSettingDetailExtension, rootWidgetClass=&ScaleBox, widgetCount=2}
 widgets [
  ScaleBox_0 {widgetClassPath=&ScaleBox, widget=ScaleBox_0, properties {stretch=ScaleBySafeZone}}
  SubtitleDisplay {
   widgetClassPath=/Script/SubtitlesWidgets.SubtitleDisplay
   bIsVariable=true
   parent=ScaleBox_0
   widget=SubtitleDisplay
   properties {
    bPreviewMode=true
    options=/Game/UI/SubtitleOptions
    previewText=SEE-Bot SEES THE INSIDE OF THE BAG. THERE IS SO MUCH BAG TO SEE.
    wrapTextAt=700
   }
   slot {class=ScaleBoxSlot, properties {horizontalAlignment=HAlign_Fill}}
  }
 ]
}
